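# Translucent, two-sided Niagara sprite material for ember particles on the Cyclone character's jacket.
# An 8x8 Husk bits SubUV flipbook, advanced by particle age plus random offset, is tinted by particle color for emissive and opacity.
# Mobile swaps the flipbook for one static, cell-sized sample.
# Attributes, including a camera-vector normal, pass through CycloneUniverse's MF_Jacket, then the shared FX emissive-adjustment function.
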Material /BRCosmetics/Characters/Player/Male/Medium/Bodies/M_MED_Cyclone/FX/M_CycloneJacket_Embers
graph {
 properties {
  bUseMaterialAttributes=true
  bUsedWithNiagaraSprites=true
  blendMode=BLEND_Translucent
  pixelDepthOffsetMode=PDOM_Legacy
  twoSided=true
  usedShadingModels=Default Lit
 }
 nodes [
  n1 = ParticleColor()
  n2 = TextureObject() {texture=/Game/Effects/Fort_Effects/Textures/Husk/T_Husk_Bits_SubUV}
  n3 = ParticleRelativeTime()
  n4 = ParticleRandom()
  n5 = MaterialFunctionCall!(
   Texture=@n2
   UVs
   SubImages=Constant2Vector() {g=8, r=8}
   Frame=Multiply(A=Add(A=@n3, B=@n4), B=ScalarParameter() {defaultValue=64, parameterName=Frame Speed})
  ) {materialFunction=/Engine/Functions/Engine_MaterialFunctions03/Texturing/SubUV_Function}
  n6 = TextureSample(UVs=TextureCoordinate() {uTiling=0.125, vTiling=0.125}, Tex=@n2) {texture=&T_Husk_Bits_SubUV}
  n7 = Multiply(A=@n1.A, B=ShadingPathSwitch(Default=@n5.Alpha, Mobile=@n6.A))
  n8 = Transform(None=CameraVectorWS()) {transformSourceType=TRANSFORMSOURCE_World, transformType=TRANSFORM_Tangent}
  n9 = ShadingPathSwitch(Default=Multiply(A=ComponentMask(None=@n5.RGB) {r=true}, B=@n1.RGB), Mobile=Multiply(A=@n6.R, B=@n1.RGB))
  MaterialFunctionCall!(In=MaterialFunctionCall!(
    Alpha=@n7
    "Gmap "=Constant() {r=1}
    Material=SetMaterialAttributes!(MaterialAttributes, Emissive Color=@n9, Opacity=@n7, Normal=@n8) {attributeSetTypes [#1, #2, #3]}
   ) {materialFunction=/Game/Characters/Player/Male/Medium/Bodies/M_MED_CycloneUniverse/Materials/MF_Jacket}.Result) {materialFunction=/Game/Effects/Niagara/MasterMaterials/Functions/MF_FX_EmissiveAdjustments_MA}
 ]
}
