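# Niagara hit burst for the Cyclone character cosmetic, a short space-themed flash with no user parameters.
# GPU Jacket_PIeces sprites burst along an X hemisphere, then swirl under curl noise, drag and upward acceleration while shrinking.
# Brief stretching Streaks share their randomly sampled blue-orange-purple gradient, boosted to HDR.
# Two single-mesh emitters, an orange Chromatic_Bubble and faint white Spherical_Density shell, expand and fade while animating dynamic material parameters.
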NiagaraSystem /BRCosmetics/Characters/Player/Male/Medium/Bodies/M_MED_Cyclone/FX/NS_Cyclone_SpaceHit
dependencies {
 usedDataInterfaces [&NiagaraDataInterfaceColorCurve, &NiagaraDataInterfaceCurve]
 usedDynamicInputs []
 usedModules [&SystemState]
}
emitters {
 Chromatic_Bubble {
  emitterUpdateScript [
   &EmitterState()
   &SpawnBurst_Instantaneous(Spawn Count=1, Use Spawn Probability=false)
  ]
  particleSpawnScript [
   /Niagara/Modules/Spawn/Initialization/InitializeParticle(
    Color=&MakeLinearColorFromVectorAndFloat(
     "Vector (RGB)"=&Multiply_Vector(
      A=&MakeVectorFromLinearColorRGB(Color=<rgba 1 0.2908 0 1>)
      B=&VectorFromFloat(Value=&UniformRangedFloat(Minimum=100, Maximum=200))
     )
     "Float (Alpha)"=2
    )
    Lifetime=&UniformRangedFloat(Minimum=0.2, Maximum=0.35)
    Mesh Scale=<1.5 1.5 1.5>
    Module.Write Color=true
    Module.Write Lifetime=true
    Module.Write Mass=true
    Module.Write Material Random=true
    Module.Write Position=false
    Module.Write Scale=true
    Module.Write SpriteRotation=false
    Module.Write SpriteSize=false
    Position=&SimulationPosition()
    Sprite Size=<5 5>
   )
   &DynamicMaterialParameters(
    Index 0 Param 2=&UniformRangedFloat(Minimum=0, Maximum=1)
    Index 0 Param 3=&UniformRangedFloat(Minimum=0, Maximum=1)
    Module.Param0WriteEnabled=false
    Module.Param1WriteEnabled=true
    Module.Param2WriteEnabled=true
    Module.Param3WriteEnabled=false
   )
   /Niagara/Modules/Spawn/Orientation/V2/InitialMeshOrientation(
    Rotation=&UniformRangedVector(Minimum=<0 0 0>, Maximum=<1 1 1>)
    Use Orientation Vector=true
    Use Rotation Vector=true
   )
  ]
  particleUpdateScript [
   &ParticleState()
   &SolveForcesAndVelocity(Clamp Velocity=false, Limit Acceleration=false)
   &DynamicMaterialParameters(
    Index 0 Param 1=&Lerp_Float(A=0.07, B=0.25, Alpha=$Particles.NormalizedAge)
    Index 0 Param 4=&Lerp_Float(A=2, B=4, Alpha=$Particles.NormalizedAge)
    Module.Param0WriteEnabled=true
    Module.Param1WriteEnabled=false
    Module.Param2WriteEnabled=false
    Module.Param3WriteEnabled=true
   ) {module=DynamicMaterialParameters001}
   &ScaleColor(Scale Alpha=&Lerp_Float(A=1, B=0, Alpha=$Particles.NormalizedAge), ScaleA=true, ScaleRGB=true)
   &ScaleMeshSize(Scale Factor=&Lerp_Vector(A=<0.5 0.5 0.5>, B=<1 1 1>, Alpha=&Power(a=$Particles.NormalizedAge, b=2)))
  ]
  rendererClasses [&NiagaraMeshRendererProperties]
  renderers [{rendererClass=&NiagaraMeshRendererProperties, rendererIndex=0}]
 }
 Jacket_PIeces {
  simTarget=GPUComputeSim
  emitterUpdateScript [
   &EmitterState(Life Cycle Mode=::NewEnumerator1, Loop Behavior=::NewEnumerator1, Loop Duration=0.3, UseLoopDelay=false)
   &SpawnBurst_Instantaneous(Spawn Count=&UniformRangedInt(Minimum=20, Maximum=30), Use Spawn Probability=false)
  ]
  particleSpawnScript [
   &InitializeParticle(
    Color=&ColorFromCurve(
     ColorCurve=DataInterface() {propertyValues {
       AlphaCurve {keys [<at 0 1>,<at 1 1>]}
       BlueCurve {keys [
         {interpMode=RCIM_Cubic, leaveTangent=0.1119, tangentMode=RCTM_Break, time=0, value=0.3333}
         {arriveTangent=-0.6438, interpMode=RCIM_Cubic, leaveTangent=-0.6438, time=0.3073, value=0.599}
         {arriveTangent=-1.6357, interpMode=RCIM_Cubic, leaveTangent=-1.6357, time=0.5178, value=0}
         {interpMode=RCIM_Cubic, tangentMode=RCTM_Break, time=0.6364, value=0.0606}
         {interpMode=RCIM_Cubic, tangentMode=RCTM_Break, time=1, value=0.349}
        ]}
       GreenCurve {keys [
         {interpMode=RCIM_Cubic, tangentMode=RCTM_Break, time=0, value=0.0901}
         {arriveTangent=0.369, interpMode=RCIM_Cubic, leaveTangent=0.369, tangentMode=RCTM_User, time=0.3073, value=0.148}
         {arriveTangent=0.0694, interpMode=RCIM_Cubic, leaveTangent=0.0694, tangentMode=RCTM_User, time=0.5178, value=0.2908}
         {arriveTangent=-0.3542, interpMode=RCIM_Cubic, leaveTangent=-0.3542, tangentMode=RCTM_User, time=0.6364, value=0}
         {arriveTangent=0.0185, interpMode=RCIM_Cubic, leaveTangent=0.0185, tangentMode=RCTM_User, time=1, value=0}
        ]}
       RedCurve {keys [
         <at 0 0.0156 RCIM_Cubic>
         {
          arriveTangent=0.3413
          arriveTangentWeight=0.4006
          interpMode=RCIM_Cubic
          leaveTangent=0.3413
          leaveTangentWeight=0.4006
          tangentMode=RCTM_Break
          time=0.3073
          value=0.0034
         }
         {arriveTangent=4.0672, interpMode=RCIM_Cubic, leaveTangent=-0.5959, tangentMode=RCTM_Break, time=0.5178, value=1}
         {arriveTangent=-3.2078, interpMode=RCIM_Cubic, leaveTangent=-3.2078, tangentMode=RCTM_User, time=0.6364, value=0.2135}
         <at 1 0.2649 RCIM_Cubic>
        ]}
      }}
     CurveIndex=&UniformRangedFloat(Minimum=0, Maximum=1)
    )
    Lifetime=&UniformRangedFloat(Minimum=0.2, Maximum=1.5)
    Module.Write Color=true
    Module.Write Lifetime=true
    Module.Write Mass=true
    Module.Write Material Random=true
    Module.Write Position=false
    Module.Write Scale=false
    Module.Write SpriteRotation=true
    Module.Write SpriteSize=true
    Position=&SimulationPosition()
    Sprite Rotation=&NormalizedAngleToDegrees(NormalizedAngle=&UniformRangedFloat(Minimum=0, Maximum=1))
    Sprite Size=&Vector2DFromFloat(Value=&UniformRangedFloat(Minimum=5, Maximum=20))
   )
   /Niagara/Modules/Spawn/Location/SphereLocation(Hemisphere X=true, MaskSpawnGroup=false, Module.Override Randomness=false, Sphere Radius=10, Surface Only=false)
   &AddVelocity(Velocity=&Multiply_Vector(A=$Output.SphereLocation.SphereNormal, B=&VectorFromFloat(Value=500)))
   &CameraOffset(Camera Offset Amount=10)
   &AddVelocity(Velocity=<200 0 0>) {module=AddVelocity001}
  ]
  particleUpdateScript [
   &ParticleState()
   &AccelerationForce(Acceleration=<0 0 100>)
   &Drag(Drag=&UniformRangedFloat(Minimum=4, Maximum=8), Use Linear Drag=true, Use Rotational Drag=false)
   /Niagara/Modules/Update/Forces/CurlNoiseForce(Noise Frequency=0.5, Noise Quality / Cost=::NewEnumerator0, Noise Strength=2500, Pan Noise=true, Pan Noise Field=<0 0 10>)
   &SolveForcesAndVelocity(Clamp Velocity=false, Limit Acceleration=false)
   &ScaleColor(
    Scale RGB=&VectorFromFloat(Value=&Lerp_Float(A=100, B=1, Alpha=&Power(a=$Particles.NormalizedAge, b=1)))
    ScaleA=true
    ScaleRGB=true
   )
   &ScaleColor(Scale RGB=<100 100 100>, ScaleA=true, ScaleRGB=true) {module=ScaleColor001}
   &ScaleSpriteSize(Scale Factor=&Vector2DFromFloat(Value=&Lerp_Float(A=1, B=0, Alpha=&Power(a=$Particles.NormalizedAge, b=1))))
  ]
  rendererClasses [&NiagaraSpriteRendererProperties]
  renderers [{rendererClass=&NiagaraSpriteRendererProperties}]
 }
 Spherical_Density {
  emitterUpdateScript [
   &EmitterState()
   &SpawnBurst_Instantaneous(Spawn Count=1, Use Spawn Probability=false)
  ]
  particleSpawnScript [
   &InitializeParticle(
    Color=<rgba 1 1 1 0.125>
    Lifetime=0.25
    Mesh Scale=<2 2 2>
    Module.Write Color=true
    Module.Write Lifetime=true
    Module.Write Mass=true
    Module.Write Material Random=true
    Module.Write Position=false
    Module.Write Scale=true
    Module.Write SpriteRotation=false
    Module.Write SpriteSize=false
    Position=&SimulationPosition()
    Sprite Size=<5 5>
   )
   &DynamicMaterialParameters(
    Index 0 Param 2=2
    Index 0 Param 4=1.25
    Module.Param0WriteEnabled=false
    Module.Param1WriteEnabled=true
    Module.Param2WriteEnabled=true
    Module.Param3WriteEnabled=true
   )
  ]
  particleUpdateScript [
   &ParticleState()
   &SolveForcesAndVelocity(Clamp Velocity=false, Limit Acceleration=false)
   &ScaleMeshSize(Scale Factor=&VectorFromFloat(Value=&Lerp_Float(A=1, B=1.25, Alpha=$Particles.NormalizedAge)))
   &DynamicMaterialParameters(
    Index 0 Param 1=&Lerp_Float(A=2, B=3, Alpha=$Particles.NormalizedAge)
    Module.Param0WriteEnabled=true
    Module.Param1WriteEnabled=false
    Module.Param2WriteEnabled=false
    Module.Param3WriteEnabled=false
   ) {module=DynamicMaterialParameters001}
   &ScaleColor(
    Scale Alpha=&FloatFromCurve(FloatCurve=DataInterface() {propertyValues {Curve {keys [<at 0 0 RCIM_Cubic>,<at 0.2 1 RCIM_Cubic>,<at 1 0 RCIM_Cubic>]}}})
    Scale RGB=&VectorFromFloat(Value=&Lerp_Float(A=1, B=0, Alpha=&Power(a=$Particles.NormalizedAge, b=2)))
    ScaleA=true
    ScaleRGB=true
   )
  ]
  rendererClasses [&NiagaraMeshRendererProperties]
  renderers [{rendererClass=&NiagaraMeshRendererProperties, rendererIndex=0}]
 }
 Streaks {
  emitterUpdateScript [
   &EmitterState()
   &SpawnBurst_Instantaneous(Spawn Count=25, Use Spawn Probability=false)
  ]
  particleSpawnScript [
   &InitializeParticle(
    Color=&MakeLinearColorFromVectorAndFloat(
     "Vector (RGB)"=&Multiply_Vector(
      A=&MakeVectorFromLinearColorRGB(Color=&ColorFromCurve(
        ColorCurve=DataInterface() {propertyValues {
          AlphaCurve {keys [<at 0 1>,<at 1 1>]}
          BlueCurve {keys [
            {interpMode=RCIM_Cubic, leaveTangent=0.1119, tangentMode=RCTM_Break, time=0, value=0.3333}
            {arriveTangent=-0.6438, interpMode=RCIM_Cubic, leaveTangent=-0.6438, time=0.3073, value=0.599}
            {arriveTangent=-1.6357, interpMode=RCIM_Cubic, leaveTangent=-1.6357, time=0.5178, value=0}
            {interpMode=RCIM_Cubic, tangentMode=RCTM_Break, time=0.6364, value=0.0606}
            {interpMode=RCIM_Cubic, tangentMode=RCTM_Break, time=1, value=0.349}
           ]}
          GreenCurve {keys [
            {interpMode=RCIM_Cubic, tangentMode=RCTM_Break, time=0, value=0.0901}
            {arriveTangent=0.369, interpMode=RCIM_Cubic, leaveTangent=0.369, tangentMode=RCTM_User, time=0.3073, value=0.148}
            {arriveTangent=0.0694, interpMode=RCIM_Cubic, leaveTangent=0.0694, tangentMode=RCTM_User, time=0.5178, value=0.2908}
            {arriveTangent=-0.3542, interpMode=RCIM_Cubic, leaveTangent=-0.3542, tangentMode=RCTM_User, time=0.6364, value=0}
            {arriveTangent=0.0185, interpMode=RCIM_Cubic, leaveTangent=0.0185, tangentMode=RCTM_User, time=1, value=0}
           ]}
          RedCurve {keys [
            <at 0 0.0156 RCIM_Cubic>
            {
             arriveTangent=0.3413
             arriveTangentWeight=0.4006
             interpMode=RCIM_Cubic
             leaveTangent=0.3413
             leaveTangentWeight=0.4006
             tangentMode=RCTM_Break
             time=0.3073
             value=0.0034
            }
            {arriveTangent=4.0672, interpMode=RCIM_Cubic, leaveTangent=-0.5959, tangentMode=RCTM_Break, time=0.5178, value=1}
            {arriveTangent=-3.2078, interpMode=RCIM_Cubic, leaveTangent=-3.2078, tangentMode=RCTM_User, time=0.6364, value=0.2135}
            <at 1 0.2649 RCIM_Cubic>
           ]}
         }}
        CurveIndex=&UniformRangedFloat(Minimum=0, Maximum=1)
       ))
      B=&VectorFromFloat(Value=&UniformRangedFloat(Minimum=10, Maximum=100))
     )
     "Float (Alpha)"=1
    )
    Lifetime=&UniformRangedFloat(Minimum=0.06, Maximum=0.2)
    Module.Write Color=true
    Module.Write Lifetime=true
    Module.Write Mass=true
    Module.Write Material Random=true
    Module.Write Position=false
    Module.Write Scale=false
    Module.Write SpriteRotation=false
    Module.Write SpriteSize=true
    Position=&SimulationPosition()
    Sprite Size=&UniformRangedVector2D(Minimum=<10 60>, Maximum=<5 80>)
   )
   &SphereLocation(Hemisphere X=true, MaskSpawnGroup=false, Module.Override Randomness=false, Sphere Radius=2, Surface Only=false)
   &AddVelocity(Velocity=&Multiply_Vector(A=$Output.SphereLocation.SphereNormal, B=&VectorFromFloat(Value=150)))
   /SetVariables(Particles.SubImageIndex=&UniformRangedFloat(Minimum=0, Maximum=3)) {isSetParameters=true}
  ]
  particleUpdateScript [
   &ParticleState()
   &SolveForcesAndVelocity(Clamp Velocity=false, Limit Acceleration=false)
   &ScaleSpriteSize(Scale Factor=&MakeVector2D(
     X=&Lerp_Float(A=0.75, B=1, Alpha=$Particles.NormalizedAge)
     Y=&Lerp_Float(A=0.03, B=2, Alpha=&Power(a=$Particles.NormalizedAge, b=5))
    ))
   &ScaleColor(
    Scale Alpha=&Lerp_Float(A=1, B=0, Alpha=$Particles.NormalizedAge)
    Scale RGB=&Lerp_Vector(A=<10 10 10>, B=<1 1 1>, Alpha=$Particles.NormalizedAge)
    ScaleA=true
    ScaleRGB=true
   )
  ]
  rendererClasses [&NiagaraSpriteRendererProperties]
  renderers [{rendererClass=&NiagaraSpriteRendererProperties}]
 }
}
summary {
 systemName=NS_Cyclone_SpaceHit
 emitters [
  {emitterName=Jacket_PIeces, simTarget=GPUComputeSim, rendererClasses [&NiagaraSpriteRendererProperties]}
  {emitterName=Streaks, rendererClasses [&NiagaraSpriteRendererProperties]}
  {emitterName=Chromatic_Bubble, rendererClasses [&NiagaraMeshRendererProperties]}
  {emitterName=Spherical_Density, rendererClasses [&NiagaraMeshRendererProperties]}
 ]
 userVariables []
}
