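# Cosmetic Niagara effect for the Cyclone "UniverseBody" character body, shedding glowing motes and faint cyan smoke from the skeletal mesh.
# Falling and rising coat emitters spawn bright, random gradient-colored sprites around skeleton samples, drifting with curl noise, with upward acceleration on the rising one.
# The smoke emitter adds growing, fading sub-image sprites driven by owner velocity and dynamic material parameters.
# User.IsSkydiving zeroes all three spawn rates while true.
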NiagaraSystem /BRCosmetics/Characters/Player/Male/Medium/Bodies/M_MED_Cyclone/FX/NS_Cyclone_UniverseBody
dependencies {
 usedDataInterfaces [&NiagaraDataInterfaceColorCurve, &NiagaraDataInterfaceCurve]
 usedDynamicInputs []
 usedModules [&SystemState]
}
emitters {
 Coat_Sheddings_Falling {
  emitterUpdateScript [
   &EmitterState()
   &SpawnRate(
    SpawnRate=&Multiply_Float(
     A=&UniformRangedFloat(Minimum=0, Maximum=10)
     B=&MakeCustomFloatFromBool(Bool=$User.IsSkydiving, True Float=0, False Float=1)
    )
    Use Spawn Probability=false
   )
  ]
  particleSpawnScript [
   /Niagara/Modules/Spawn/Initialization/InitializeParticle(
    Color=&MultiplyLinearColor(
     Linear Color=&ColorFromCurve(
      ColorCurve=DataInterface() {propertyValues {
        AlphaCurve {keys [<at 0 1>,<at 1 1>]}
        BlueCurve {keys [
          {interpMode=RCIM_Cubic, leaveTangent=0.1119, tangentMode=RCTM_Break, time=0, value=0.3333}
          {arriveTangent=-0.6438, interpMode=RCIM_Cubic, leaveTangent=-0.6438, time=0.3073, value=0.599}
          {arriveTangent=-1.6357, interpMode=RCIM_Cubic, leaveTangent=-1.6357, time=0.5178, value=0}
          {interpMode=RCIM_Cubic, tangentMode=RCTM_Break, time=0.6364, value=0.0606}
          {interpMode=RCIM_Cubic, tangentMode=RCTM_Break, time=1, value=0.349}
         ]}
        GreenCurve {keys [
          {interpMode=RCIM_Cubic, tangentMode=RCTM_Break, time=0, value=0.0901}
          {arriveTangent=0.369, interpMode=RCIM_Cubic, leaveTangent=0.369, tangentMode=RCTM_User, time=0.3073, value=0.148}
          {arriveTangent=0.0694, interpMode=RCIM_Cubic, leaveTangent=0.0694, tangentMode=RCTM_User, time=0.5178, value=0.2908}
          {arriveTangent=-0.3542, interpMode=RCIM_Cubic, leaveTangent=-0.3542, tangentMode=RCTM_User, time=0.6364, value=0}
          {arriveTangent=0.0185, interpMode=RCIM_Cubic, leaveTangent=0.0185, tangentMode=RCTM_User, time=1, value=0}
         ]}
        RedCurve {keys [
          <at 0 0.0156 RCIM_Cubic>
          {
           arriveTangent=0.3413
           arriveTangentWeight=0.4006
           interpMode=RCIM_Cubic
           leaveTangent=0.3413
           leaveTangentWeight=0.4006
           tangentMode=RCTM_Break
           time=0.3073
           value=0.0034
          }
          {arriveTangent=4.0672, interpMode=RCIM_Cubic, leaveTangent=-0.5959, tangentMode=RCTM_Break, time=0.5178, value=1}
          {arriveTangent=-3.2078, interpMode=RCIM_Cubic, leaveTangent=-3.2078, tangentMode=RCTM_User, time=0.6364, value=0.2135}
          <at 1 0.2649 RCIM_Cubic>
         ]}
       }}
      CurveIndex=&UniformRangedFloat(Minimum=0, Maximum=1)
     )
     Scale Factor=<5 5 5 1>
    )
    Lifetime=&UniformRangedFloat(Minimum=0.75, Maximum=1.5)
    Module.Write Color=true
    Module.Write Lifetime=true
    Module.Write Mass=true
    Module.Write Material Random=true
    Module.Write Position=false
    Module.Write Scale=false
    Module.Write SpriteRotation=true
    Module.Write SpriteSize=true
    Position=&SimulationPosition()
    Sprite Rotation=&NormalizedAngleToDegrees(NormalizedAngle=&UniformRangedFloat(Minimum=0, Maximum=1))
    Sprite Size=&Vector2DFromFloat(Value=&UniformRangedFloat(Minimum=5, Maximum=10))
   )
   &SampleSkeletalMeshSkeleton(Mesh:NiagaraDataInterfaceSkeletalMesh=$User.SkeletalMesh)
   /Niagara/Modules/Spawn/Location/SkeletalMeshSkeletonLocation(Inherit Velocity From Mesh=0.5, Module.Inherit Velocity=true, Module.Use Sampled Bone as Ribbon ID=true)
   /Niagara/Modules/Spawn/Location/SphereLocation(MaskSpawnGroup=false, Module.Override Randomness=false, Non Uniform Scale=<-1 -1 -1>, Sphere Radius=10, Surface Only=false)
   &AddVelocity(Velocity=&MakeVector(X=0, Y=0, Z=-50))
   &CameraOffset(Camera Offset Amount=50)
  ]
  particleUpdateScript [
   &ParticleState()
   &Drag(Use Linear Drag=true, Use Rotational Drag=false)
   /Niagara/Modules/Update/Forces/CurlNoiseForce(Noise Frequency=0.5, Noise Strength=200, Pan Noise=true, Pan Noise Field=<0 0 -50>)
   &SolveForcesAndVelocity(Clamp Velocity=false, Limit Acceleration=false)
   &ScaleColor(Scale RGB=<100 100 100>, ScaleA=true, ScaleRGB=true)
   &ScaleSpriteSize(Scale Factor=&Vector2DFromFloat(Value=&FloatFromCurve(FloatCurve=DataInterface() {propertyValues {Curve {keys [
          {arriveTangent=20.0357, interpMode=RCIM_Cubic, leaveTangent=20.0357, tangentMode=RCTM_User, time=0, value=0}
          <at 0.15 1 RCIM_Cubic>
          {arriveTangent=-3.9887, interpMode=RCIM_Cubic, leaveTangent=-3.9887, tangentMode=RCTM_User, time=1, value=0}
         ]}}})))
  ]
  rendererClasses [&NiagaraSpriteRendererProperties]
  renderers [{rendererClass=&NiagaraSpriteRendererProperties}]
 }
 Coat_Sheddings_Rising {
  emitterUpdateScript [
   &EmitterState()
   &SpawnRate(
    SpawnRate=&Multiply_Float(
     A=&UniformRangedFloat(Minimum=0, Maximum=10)
     B=&MakeCustomFloatFromBool(Bool=$User.IsSkydiving, True Float=0, False Float=1)
    )
    Use Spawn Probability=false
   )
  ]
  particleSpawnScript [
   &InitializeParticle(
    Color=&MultiplyLinearColor(
     Linear Color=&ColorFromCurve(
      ColorCurve=DataInterface() {propertyValues {
        AlphaCurve {keys [<at 0 1>,<at 1 1>]}
        BlueCurve {keys [
          {interpMode=RCIM_Cubic, leaveTangent=0.1119, tangentMode=RCTM_Break, time=0, value=0.3333}
          {arriveTangent=-0.6438, interpMode=RCIM_Cubic, leaveTangent=-0.6438, time=0.3073, value=0.599}
          {arriveTangent=-1.6357, interpMode=RCIM_Cubic, leaveTangent=-1.6357, time=0.5178, value=0}
          {interpMode=RCIM_Cubic, tangentMode=RCTM_Break, time=0.6364, value=0.0606}
          {interpMode=RCIM_Cubic, tangentMode=RCTM_Break, time=1, value=0.349}
         ]}
        GreenCurve {keys [
          {interpMode=RCIM_Cubic, tangentMode=RCTM_Break, time=0, value=0.0901}
          {arriveTangent=0.369, interpMode=RCIM_Cubic, leaveTangent=0.369, tangentMode=RCTM_User, time=0.3073, value=0.148}
          {arriveTangent=0.0694, interpMode=RCIM_Cubic, leaveTangent=0.0694, tangentMode=RCTM_User, time=0.5178, value=0.2908}
          {arriveTangent=-0.3542, interpMode=RCIM_Cubic, leaveTangent=-0.3542, tangentMode=RCTM_User, time=0.6364, value=0}
          {arriveTangent=0.0185, interpMode=RCIM_Cubic, leaveTangent=0.0185, tangentMode=RCTM_User, time=1, value=0}
         ]}
        RedCurve {keys [
          <at 0 0.0156 RCIM_Cubic>
          {
           arriveTangent=0.3413
           arriveTangentWeight=0.4006
           interpMode=RCIM_Cubic
           leaveTangent=0.3413
           leaveTangentWeight=0.4006
           tangentMode=RCTM_Break
           time=0.3073
           value=0.0034
          }
          {arriveTangent=4.0672, interpMode=RCIM_Cubic, leaveTangent=-0.5959, tangentMode=RCTM_Break, time=0.5178, value=1}
          {arriveTangent=-3.2078, interpMode=RCIM_Cubic, leaveTangent=-3.2078, tangentMode=RCTM_User, time=0.6364, value=0.2135}
          <at 1 0.2649 RCIM_Cubic>
         ]}
       }}
      CurveIndex=&UniformRangedFloat(Minimum=0, Maximum=1)
     )
     Scale Factor=<5 5 5 1>
    )
    Lifetime=&UniformRangedFloat(Minimum=0.75, Maximum=1.5)
    Module.Write Color=true
    Module.Write Lifetime=true
    Module.Write Mass=true
    Module.Write Material Random=true
    Module.Write Position=false
    Module.Write Scale=false
    Module.Write SpriteRotation=true
    Module.Write SpriteSize=true
    Position=&SimulationPosition()
    Sprite Rotation=&NormalizedAngleToDegrees(NormalizedAngle=&UniformRangedFloat(Minimum=0, Maximum=1))
    Sprite Size=&Vector2DFromFloat(Value=&UniformRangedFloat(Minimum=5, Maximum=10))
   )
   &SampleSkeletalMeshSkeleton(Mesh:NiagaraDataInterfaceSkeletalMesh=$User.SkeletalMesh)
   &SkeletalMeshSkeletonLocation(Inherit Velocity From Mesh=0.5, Module.Inherit Velocity=true, Module.Use Sampled Bone as Ribbon ID=true)
   &SphereLocation(Hemisphere Z=true, MaskSpawnGroup=false, Module.Override Randomness=false, Sphere Radius=10, Surface Only=false)
   &AddVelocity(Velocity=&MakeVector(X=0, Y=0, Z=-50))
   &CameraOffset(Camera Offset Amount=50)
  ]
  particleUpdateScript [
   &ParticleState()
   &AccelerationForce(Acceleration=<0 0 125>)
   &Drag(Drag=&UniformRangedFloat(Minimum=0, Maximum=2), Use Linear Drag=true, Use Rotational Drag=false)
   &CurlNoiseForce(Noise Frequency=0.5, Noise Strength=500, Pan Noise=true, Pan Noise Field=<0 0 -50>)
   &SolveForcesAndVelocity(Clamp Velocity=false, Limit Acceleration=false)
   &ScaleColor(Scale RGB=<100 100 100>, ScaleA=true, ScaleRGB=true)
   &ScaleSpriteSize(Scale Factor=&Vector2DFromFloat(Value=&FloatFromCurve(FloatCurve=DataInterface() {propertyValues {Curve {keys [
          {arriveTangent=20.0357, interpMode=RCIM_Cubic, leaveTangent=20.0357, tangentMode=RCTM_User, time=0, value=0}
          <at 0.15 1 RCIM_Cubic>
          {arriveTangent=-3.9887, interpMode=RCIM_Cubic, leaveTangent=-3.9887, tangentMode=RCTM_User, time=1, value=0}
         ]}}})))
  ]
  rendererClasses [&NiagaraSpriteRendererProperties]
  renderers [{rendererClass=&NiagaraSpriteRendererProperties}]
 }
 Soft_Wispy_Particulate_Smoke {
  emitterUpdateScript [
   &EmitterState(Life Cycle Mode=::NewEnumerator1, Loop Duration=1, UseLoopDelay=false)
   &SpawnRate(
    SpawnRate=&Multiply_Float(A=8, B=&MakeCustomFloatFromBool(Bool=$User.IsSkydiving, True Float=0, False Float=1))
    Use Spawn Probability=false
   )
  ]
  particleSpawnScript [
   &InitializeParticle(
    Color=<rgba 1 1 1 0.0125>
    Lifetime=&UniformRangedFloat(Minimum=0.5, Maximum=1)
    Module.Write Color=true
    Module.Write Lifetime=true
    Module.Write Mass=true
    Module.Write Material Random=true
    Module.Write Position=false
    Module.Write Scale=false
    Module.Write SpriteRotation=true
    Module.Write SpriteSize=true
    Position=&SimulationPosition()
    Sprite Rotation=&UniformRangedFloat(Minimum=0, Maximum=360)
    Sprite Size=&Vector2DFromFloat(Value=&UniformRangedFloat(Minimum=75, Maximum=75))
   )
   &SampleSkeletalMeshSkeleton(Mesh:NiagaraDataInterfaceSkeletalMesh=$User.SkeletalMesh, Specific Bone Index=0)
   &SkeletalMeshSkeletonLocation(
    Module.Inherit Velocity=false
    Module.Use Sampled Bone as Ribbon ID=true
    Sampled Position Offset:Vector3f=&TransformVector(Vector=<0 50 0>, Destination Space=::Simulation)
   )
   &AddVelocity(Velocity=&UniformRangedVector(Minimum=<-50 -50 -100>, Maximum=<50 50 -100>))
   &InheritVelocity(
    Module.Limit Speed=true
    Velocity Scale=<-0.3 -0.3 0.1>
    Velocity Source=&TransformVector(Vector=$Engine.Owner.Velocity, Source Space=::World, Destination Space=::Local, Apply Scale=true)
   )
   &DynamicMaterialParameters(
    Index 0 Param 1=1
    Module.Param0WriteEnabled=true
    Module.Param1WriteEnabled=false
    Module.Param2WriteEnabled=false
    Module.Param3WriteEnabled=true
   ) {module=DynamicMaterialParameters001}
  ]
  particleUpdateScript [
   &ParticleState()
   &SolveForcesAndVelocity(Clamp Velocity=false, Limit Acceleration=false)
   /SetVariables(Particles.SubImageIndex=&Lerp_Float(A=0, B=63, Alpha=&Power(a=$Particles.NormalizedAge, b=100))) {isSetParameters=true}
   &ScaleSpriteSize(Scale Factor=&Vector2DFromFloat(Value=&Lerp_Float(A=1, B=3, Alpha=$Particles.NormalizedAge)))
   &ScaleColor(
    Scale Alpha=&FloatFromCurve(FloatCurve=DataInterface() {propertyValues {Curve {keys [<at 0 0 RCIM_Cubic>,<at 0.2 1 RCIM_Cubic>,<at 1 0 RCIM_Cubic>]}}})
    Scale RGB=&MakeVectorFromLinearColorRGB(Color=<rgba 0 1 0.9703 1>)
    ScaleA=true
    ScaleRGB=true
   )
   &DynamicMaterialParameters(
    Index 0 Param 1=2
    Index 0 Param 2=&Lerp_Float(A=2, B=0, Alpha=$Particles.NormalizedAge)
    Index 0 Param 3=&Lerp_Float(A=2, B=1, Alpha=$Particles.NormalizedAge)
    Index 0 Param 4=1
    Module.Param0WriteEnabled=false
    Module.Param1WriteEnabled=true
    Module.Param2WriteEnabled=true
    Module.Param3WriteEnabled=false
   )
  ]
  rendererClasses [&NiagaraSpriteRendererProperties]
  renderers [{rendererClass=&NiagaraSpriteRendererProperties}]
 }
}
summary {
 systemName=NS_Cyclone_UniverseBody
 emitters [
  {emitterName=Coat_Sheddings_Rising, rendererClasses [&NiagaraSpriteRendererProperties]}
  {emitterName=Coat_Sheddings_Falling, rendererClasses [&NiagaraSpriteRendererProperties]}
  {emitterName=Soft_Wispy_Particulate_Smoke, rendererClasses [&NiagaraSpriteRendererProperties]}
 ]
 userVariables [
  User.IsSkydiving {type=NiagaraBool}
  User.SkeletalMesh {
   type=NiagaraDataInterfaceSkeletalMesh
   defaultValue {
    struct=NiagaraExt_VariableValue_DataInterface
    value {dataInterface=&self:NiagaraDataInterfaceSkeletalMesh_0}
   }
  }
 ]
}
