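# Niagara cosmetic effect for the Cyclone character's head: two roller coaster car meshes with lights orbit it, dressed with fiery ribbon, glow, flame and spark emitters.
# A custom orbit module uses a per-particle ID flag to give each car its own user axis and spin direction.
# Cars broadcast location events with orbit position and backward velocity.
# All emitters cull by User.MinDistance; all but sparks set VisibilityTag from bisHeadHidden.
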NiagaraSystem /BRCosmetics/Characters/Player/Male/Medium/Bodies/M_MED_Cyclone/FX/NS_Cyclone_UniverseHead
dependencies {
 usedDataInterfaces [&NiagaraDataInterfaceCurve, &NiagaraDataInterfaceColorCurve]
 usedDynamicInputs []
 usedModules [&SystemState]
}
emitters {
 Coaster_Car_AirGlow {
  emitterUpdateScript [
   &EmitterScalability(
    Module.bApplySpawnCountScale=false
    Module.bApplySpawnCountScaleByDistanceFraction=false
    Module.bUseMaxDistance=false
    Module.bUseMinDistance=false
   ) {enabled=false}
   &EmitterLifeCycle(NextLoopDuration=1) {enabled=false}
   &EmitterState(
    Enable Distance Culling=true
    Min Distance Response=::SleepAndClearParticles
    MinDistance:NiagaraFloat=$User.MinDistance
    Module.bUseMaxDistance=false
    Module.bUseMinDistance=true
    Scalability Mode=::NewEnumerator1
   )
   &SpawnRate(SpawnRate=0, Use Spawn Probability=false) {enabled=false}
   &SpawnBurst_Instantaneous(Use Spawn Probability=false)
  ]
  particleSpawnScript [/Niagara/Modules/Spawn/Initialization/InitializeParticle(
    Color=<rgba 20 2.6667 0 1>
    Lifetime=0.5
    Module.Write Color=true
    Module.Write Lifetime=true
    Module.Write Mass=true
    Module.Write Material Random=true
    Module.Write Position=true
    Module.Write Scale=false
    Module.Write SpriteRotation=false
    Module.Write SpriteSize=true
    Position=&SimulationPosition()
    Sprite Size=<7.8 5.25>
   )]
  particleUpdateScript [
   &ParticleState()
   &ScaleColor(
    Scale Alpha=&FloatFromCurve(FloatCurve=DataInterface() {propertyValues {Curve {keys [{arriveTangent=-2.941, interpMode=RCIM_Cubic, leaveTangent=-2.941, tangentMode=RCTM_User, time=0, value=1}, <at 1 0 RCIM_Cubic>]}}})
    Scale RGB=&MakeVectorFromLinearColorRGB(Color=<rgba 1 1 1 1>)
    ScaleA=true
    ScaleRGB=true
   )
   &ScaleSpriteSize(Scale Factor=&Lerp_Vector2(A=<1.1 1.1>, B=<0.2 0.2>, Alpha=$Particles.NormalizedAge))
   &CameraOffset(Camera Offset Amount=-5)
   &AccelerationForce(Acceleration=<0 0 -5>)
   &SolveForcesAndVelocity(Clamp Velocity=false, Limit Acceleration=false)
   /SetVariables(Particles.VisibilityTag=&MakeCustomIntFromBool(Bool=$User.bisHeadHidden, True INT=1, False INT=0)) {isSetParameters=true, module=SetVariables_#1}
  ]
  rendererClasses [&NiagaraSpriteRendererProperties]
  renderers [{rendererClass=&NiagaraSpriteRendererProperties}]
 }
 Coaster_Sparks_A_Cross {
  emitterUpdateScript [
   &EmitterState(
    Enable Distance Culling=true
    Life Cycle Mode=::NewEnumerator1
    Loop Duration=1
    Min Distance Response=::SleepAndClearParticles
    MinDistance:NiagaraFloat=$User.MinDistance
    Module.bUseMaxDistance=false
    Module.bUseMinDistance=true
    Scalability Mode=::NewEnumerator1
    UseLoopDelay=false
   )
   &SpawnBurst_Instantaneous(Use Spawn Probability=false)
  ]
  particleSpawnScript [
   &InitializeParticle(
    Color=&UniformRangedLinearColor(Minimum=<rgba 1 0 0 1>, Maximum=<rgba 1 0.4 0 1>)
    Lifetime=&UniformRangedFloat(Minimum=0.5, Maximum=2)
    Module.Write Color=true
    Module.Write Lifetime=true
    Module.Write Mass=true
    Module.Write Material Random=true
    Module.Write Position=false
    Module.Write Scale=false
    Module.Write SpriteRotation=false
    Module.Write SpriteSize=true
    Position=&SimulationPosition()
    Sprite Size=&Vector2DFromFloat(Value=&UniformRangedFloat(Minimum=3, Maximum=5))
   )
   /Niagara/Modules/Update/Mass/CalculateSizeAndRotationalInertiaByMass(Depth=0, Height=5, Module.Manually Enter Density=false) {enabled=false}
  ]
  particleUpdateScript [
   &ParticleState()
   &SpriteRotationRate(
    Delta Time:NiagaraFloat=&Multiply_Float(
     A=$Engine.DeltaTime
     B=&FloatFromCurve(FloatCurve=DataInterface() {propertyValues {Curve {keys [<at 0 1 RCIM_Cubic>,<at 1 0 RCIM_Cubic>]}}})
    )
    Rotation Rate=&NormalizedAngleToDegrees(NormalizedAngle=1)
   )
   &Drag(Drag=&UniformRangedFloat(Minimum=20, Maximum=50), Use Linear Drag=true, Use Rotational Drag=false)
   &ScaleColor(
    Scale Alpha=&FloatFromCurve(FloatCurve=DataInterface() {propertyValues {Curve {keys [<at -0.0046 1.0113 RCIM_Cubic>,<at 1.0011 0.002 RCIM_Cubic>]}}})
    Scale RGB=<10 10 10>
    ScaleA=true
    ScaleRGB=true
   )
   /Niagara/Modules/Update/Forces/CurlNoiseForce(Noise Frequency=0.5, Noise Strength=20, Pan Noise=true)
   &SolveForcesAndVelocity(Clamp Velocity=false, Limit Acceleration=false)
   &DynamicMaterialParameters(
    Index 0 Param 1=&FloatFromCurve(FloatCurve=DataInterface() {propertyValues {Curve {keys [{arriveTangent=-2.865, interpMode=RCIM_Cubic, leaveTangent=-2.865, tangentMode=RCTM_User, time=0, value=1}, <at 1 0 RCIM_Cubic>]}}})
    Module.Param0WriteEnabled=true
    Module.Param1WriteEnabled=false
    Module.Param2WriteEnabled=false
    Module.Param3WriteEnabled=false
   ) {module=DynamicMaterialParameters001}
  ]
  rendererClasses [&NiagaraSpriteRendererProperties]
  renderers [{rendererClass=&NiagaraSpriteRendererProperties}]
 }
 Flames_A {
  emitterUpdateScript [
   &EmitterScalability(
    Module.bApplySpawnCountScale=false
    Module.bApplySpawnCountScaleByDistanceFraction=false
    Module.bUseMaxDistance=false
    Module.bUseMinDistance=false
   ) {enabled=false}
   &EmitterLifeCycle(NextLoopDuration=1) {enabled=false}
   &EmitterState(
    Enable Distance Culling=true
    Life Cycle Mode=::NewEnumerator1
    Loop Duration=0.5
    Min Distance Response=::SleepAndClearParticles
    MinDistance:NiagaraFloat=$User.MinDistance
    Module.bUseMaxDistance=false
    Module.bUseMinDistance=true
    Scalability Mode=::NewEnumerator1
    UseLoopDelay=false
   )
   &SpawnBurst_Instantaneous(Use Spawn Probability=false)
  ]
  particleSpawnScript [&InitializeParticle(
    Color=<rgba 10 10 10 0.25>
    Lifetime=&UniformRangedFloat(Minimum=0.1, Maximum=0.75)
    Module.Write Color=true
    Module.Write Lifetime=true
    Module.Write Mass=true
    Module.Write Material Random=true
    Module.Write Position=true
    Module.Write Scale=false
    Module.Write SpriteRotation=true
    Module.Write SpriteSize=true
    Position=&SimulationPosition()
    Sprite Rotation=&UniformRangedFloat(Minimum=0, Maximum=360)
    Sprite Size=&Vector2DFromFloat(Value=&UniformRangedFloat(Minimum=5, Maximum=12))
   )]
  particleUpdateScript [
   &ParticleState()
   &DynamicMaterialParameters(
    Index 0 Param 1=1
    Index 0 Param 3=&Lerp_Float(A=0.5, B=1, Alpha=$Particles.NormalizedAge)
    Index 0 Param 4=2
    Module.Param0WriteEnabled=true
    Module.Param1WriteEnabled=true
    Module.Param2WriteEnabled=true
    Module.Param3WriteEnabled=true
   )
   &ScaleColor(
    Scale Alpha=&Lerp_Float(A=1, B=0, Alpha=$Particles.NormalizedAge)
    Scale RGB=&MakeVectorFromLinearColorRGB(Color=&ColorFromCurve(ColorCurve=DataInterface() {propertyValues {
        AlphaCurve {keys [<at 0 1 RCIM_Cubic>,<at 1 1 RCIM_Cubic>,<at 1 1 RCIM_Cubic>]}
        BlueCurve {keys [<at 0 0.73 RCIM_Cubic>, {arriveTangent=-0.73, interpMode=RCIM_Cubic, leaveTangent=-0.73, time=0.75, value=0}, <at 1 0 RCIM_Cubic>]}
        GreenCurve {keys [<at 0 1.7975 RCIM_Cubic>, {interpMode=RCIM_Cubic, tangentMode=RCTM_User, time=0.75, value=0.1667}, <at 1 0 RCIM_Cubic>]}
        RedCurve {keys [<at 0 5 RCIM_Cubic>, {interpMode=RCIM_Cubic, tangentMode=RCTM_User, time=0.75, value=10}, <at 1 0.001 RCIM_Cubic>]}
       }}))
    ScaleA=true
    ScaleRGB=true
   )
   &AccelerationForce(Acceleration=<0 0 -10>)
   &SolveForcesAndVelocity(Clamp Velocity=false, Limit Acceleration=false)
   &ScaleSpriteSize(Scale Factor=&Lerp_Vector2(A=<0.5 0.5>, B=<1.5 1.5>, Alpha=$Particles.NormalizedAge))
   &SpriteRotationRate(
    Delta Time:NiagaraFloat=&Multiply_Float(
     A=$Engine.DeltaTime
     B=&FloatFromCurve(FloatCurve=DataInterface() {propertyValues {Curve {keys [{arriveTangent=-3.2062, interpMode=RCIM_Cubic, leaveTangent=-3.2062, tangentMode=RCTM_User, time=0, value=1}, <at 1 0 RCIM_Cubic>]}}})
    )
    Rotation Rate=&NormalizedAngleToDegrees(NormalizedAngle=1)
   )
   /Niagara/Modules/Update/SubUV/SubUVAnimation(Number Of Frames=64, SubUV Lookup Index=&Multiply_Float(A=$Particles.NormalizedAge, B=0.1))
   &CameraOffset()
   /SetVariables(Particles.VisibilityTag=&MakeCustomIntFromBool(Bool=$User.bisHeadHidden, True INT=1, False INT=0)) {isSetParameters=true, module=SetVariables_#1}
  ]
  rendererClasses [&NiagaraSpriteRendererProperties]
  renderers [{rendererClass=&NiagaraSpriteRendererProperties}]
 }
 LocationBasedRibbon_A {
  emitterUpdateScript [&EmitterState(
    Enable Distance Culling=true
    Min Distance Response=::SleepAndClearParticles
    MinDistance:NiagaraFloat=$User.MinDistance
    Module.bUseMaxDistance=false
    Module.bUseMinDistance=true
    Scalability Mode=::NewEnumerator1
   )]
  particleSpawnScript [/Niagara/Modules/Spawn/Initialization/V2/InitializeParticle(UsePositionOffset=false, Write Material Random=true)]
  particleUpdateScript [
   &ParticleState()
   &AccelerationForce(Acceleration=<0 0 -5>)
   &SolveForcesAndVelocity(Clamp Velocity=false, Limit Acceleration=false)
   /SetVariables(Particles.RibbonFacing=$User.AxisB) {isSetParameters=true}
   &DynamicMaterialParameters(
    Index 0 Param 2=&FloatFromCurve(FloatCurve=DataInterface() {propertyValues {Curve {keys [{arriveTangent=-5.7575, interpMode=RCIM_Cubic, leaveTangent=-5.7575, tangentMode=RCTM_User, time=0, value=1}, <at 0.5 0 RCIM_Cubic>]}}})
    Index 0 Param 3=&FloatFromCurve(FloatCurve=DataInterface() {propertyValues {Curve {keys [<at 0.5 1 RCIM_Cubic>, {arriveTangent=45.9563, interpMode=RCIM_Cubic, leaveTangent=45.956, tangentMode=RCTM_User, time=1, value=10}]}}})
    Module.Param0WriteEnabled=true
    Module.Param1WriteEnabled=true
    Module.Param2WriteEnabled=true
    Module.Param3WriteEnabled=true
   )
   &ScaleColor(
    Scale Alpha=&FloatFromCurve(FloatCurve=DataInterface() {propertyValues {Curve {keys [
         {arriveTangent=-1.7682, interpMode=RCIM_Cubic, leaveTangent=-1.7682, tangentMode=RCTM_User, time=-0.0046, value=1.0113}
         <at 1.0011 0.002 RCIM_Cubic>
        ]}}})
    Scale RGB=&MakeVectorFromLinearColorRGB(Color=&ColorFromCurve(ColorCurve=DataInterface() {propertyValues {
        AlphaCurve {keys [<at 0 1 RCIM_Cubic>,<at 1 1 RCIM_Cubic>,<at 1 1 RCIM_Cubic>]}
        BlueCurve {keys [<at 0 0 RCIM_Cubic>,<at 0.75 0 RCIM_Cubic>,<at 1 0 RCIM_Cubic>]}
        GreenCurve {keys [<at 0 1.25 RCIM_Cubic>, {interpMode=RCIM_Cubic, tangentMode=RCTM_User, time=0.75, value=0.1667}, <at 1 0 RCIM_Cubic>]}
        RedCurve {keys [<at 0 5 RCIM_Cubic>, {interpMode=RCIM_Cubic, tangentMode=RCTM_User, time=0.75, value=10}, <at 1 0.001 RCIM_Cubic>]}
       }}))
    ScaleA=true
    ScaleRGB=true
   )
   &ScaleColor(
    Scale Alpha=0.2
    Scale RGB=&VectorFromFloat(Value=&FloatFromCurve(FloatCurve=DataInterface() {propertyValues {Curve {keys [{arriveTangent=-45.8622, interpMode=RCIM_Cubic, leaveTangent=-45.8622, tangentMode=RCTM_User, time=0, value=25}, <at 1 1>]}}}))
    ScaleA=true
    ScaleRGB=true
   ) {module=ScaleColor001}
   &ScaleRibbonWidth(Ribbon Width Scale=&FloatFromCurve(FloatCurve=DataInterface() {propertyValues {Curve {keys [<at 0 1>,<at 1 3>]}}}))
   /SetVariables(Particles.VisibilityTag=&MakeCustomIntFromBool(Bool=$User.bisHeadHidden, True INT=1, False INT=0)) {isSetParameters=true}
  ]
  rendererClasses [&NiagaraRibbonRendererProperties]
  renderers [{rendererClass=&NiagaraRibbonRendererProperties}]
 }
 Orbitting_Coaster_Car_A {
  emitterUpdateScript [
   &EmitterState(
    Enable Distance Culling=true
    Inactive Response=::NewEnumerator1
    Life Cycle Mode=::NewEnumerator1
    Loop Behavior=::NewEnumerator1
    Loop Duration Mode=::NewEnumerator1
    Min Distance Response=::SleepAndLetParticlesFinish
    MinDistance:NiagaraFloat=$User.MinDistance
    Module.bUseMaxDistance=false
    Module.bUseMinDistance=true
    Scalability Mode=::NewEnumerator1
    UseLoopDelay=false
   )
   &SpawnBurst_Instantaneous(Spawn Count=2, Use Spawn Probability=false)
  ]
  particleSpawnScript [
   /Game/Effects/Niagara/Modules/Orbit/FN_Orbit_Init_Variables_v2()
   /Niagara/Modules/Spawn/Initialization/InitializeParticle(
    Mesh Scale=<0.6 0.6 0.6>
    Module.Write Color=true
    Module.Write Lifetime=true
    Module.Write Mass=true
    Module.Write Material Random=true
    Module.Write Position=false
    Module.Write Scale=true
    Module.Write SpriteRotation=false
    Module.Write SpriteSize=false
    Position=&SimulationPosition()
   )
   /SetVariables(Particles.IsEven=HlslExpression() {hlslExpression="Particles.UniqueID && 1"}) {isSetParameters=true}
   /SetVariables(Particles.LightColor=<rgba 20 5 2 1>, Particles.LightExponent=4, Particles.LightRadius=15) {isSetParameters=true}
   &DynamicMaterialParameters(
    Index 0 Param 1=1
    Index 0 Param 3=0.5
    Module.Param0WriteEnabled=true
    Module.Param1WriteEnabled=true
    Module.Param2WriteEnabled=true
    Module.Param3WriteEnabled=true
   )
  ]
  particleUpdateScript [
   &ParticleState(Kill Particles When Lifetime Has Elapsed=false, Let Infinitely Lived Particles Die When Emitter Deactivates=true)
   &ScaleColor(ScaleA=true, ScaleRGB=true)
   /Game/Effects/Niagara/Modules/Orbit/FN_Orbit_v3(
    Delta:NiagaraFloat=$Particles.NormalizedAge
    Mesh Reference Forward Axis:Vector3f=&Lerp_Vector(A=<1 0 0>, B=<1 0 0>, Alpha=$Particles.IsEven)
    Radius min/max:Vector2f=&Vector2DFromFloat(Value=16)
    Rotation Rate min/max:Vector2f=&Lerp_Vector2(A=<0.5 0.5>, B=<-0.5 -0.5>, Alpha=$Particles.IsEven)
    To Axis:Vector3f=&Lerp_Vector(A=$User.AxisA, B=$User.AxisB, Alpha=$Particles.IsEven)
   ) {enabled=false}
   /Game/Effects/Niagara/Modules/Orbit/FN_Orbit_v4(
    Delta:NiagaraFloat=$Particles.NormalizedAge
    Mesh Reference Forward Axis:Vector3f=&Lerp_Vector(A=<1 0 0>, B=<1 0 0>, Alpha=$Particles.IsEven)
    Radius min/max:Vector2f=&Vector2DFromFloat(Value=16)
    Rotation Rate min/max:Vector2f=&Lerp_Vector2(A=<0.5 0.5>, B=<-0.5 -0.5>, Alpha=$Particles.IsEven)
    To Axis:Vector3f=&Lerp_Vector(A=$User.AxisA, B=$User.AxisB, Alpha=$Particles.IsEven)
   )
   &SolveForcesAndVelocity(Clamp Velocity=false, Limit Acceleration=false)
   &GenerateLocationEvent(
    Event Send Rate=2
    Module.Use Event Send Rate=false
    Position To Send:Vector3f=$Particles.FN_Orbit_v4.Position
    Velocity To Send:Vector3f=&Multiply_VectorByFloat(Vector=&RotateVectorByQuaternion(Vector To Rotate=<1 0 0>, Quaternion=$Particles.MeshOrientation), Float=-1)
   )
   /SetVariables(Particles.VisibilityTag=&MakeCustomIntFromBool(Bool=$User.bisHeadHidden, True INT=1, False INT=0)) {isSetParameters=true, module=SetVariables_#1}
  ]
  rendererClasses [&NiagaraLightRendererProperties, &NiagaraMeshRendererProperties]
  renderers [
   {rendererClass=&NiagaraLightRendererProperties, rendererIndex=0}
   {rendererClass=&NiagaraMeshRendererProperties, rendererIndex=1}
  ]
 }
}
summary {
 systemName=NS_Cyclone_UniverseHead
 emitters [
  {
   emitterName=Orbitting_Coaster_Car_A
   rendererClasses [&NiagaraLightRendererProperties, &NiagaraMeshRendererProperties]
  }
  {emitterName=LocationBasedRibbon_A, rendererClasses [&NiagaraRibbonRendererProperties]}
  {emitterName=Coaster_Car_AirGlow, rendererClasses [&NiagaraSpriteRendererProperties]}
  {emitterName=Flames_A, rendererClasses [&NiagaraSpriteRendererProperties]}
  {emitterName=Coaster_Sparks_A_Cross, rendererClasses [&NiagaraSpriteRendererProperties]}
 ]
 userVariables [
  User.AxisA {type=Vector3f, description="", defaultValue {struct=Vector3f, value=<0.3 0 1>}}
  User.AxisB {type=Vector3f, description="", defaultValue {struct=Vector3f, value=<-0.4 0 1>}}
  User.bisHeadHidden {type=NiagaraBool}
  User.MinDistance {type=NiagaraFloat, defaultValue=10, description=""}
 ]
}
