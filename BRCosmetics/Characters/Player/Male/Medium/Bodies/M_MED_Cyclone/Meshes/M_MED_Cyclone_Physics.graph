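# Physics asset for the medium male Cyclone player body, covering dangling leg pieces and the leg bodies around them.
# Capsules on thighs and calves form leg shapes, with tiny box bodies linking each dangle root to its calf.
# Top and bottom two-segment capsule chains branch from each root with tight twist limits.
# Left chains swing mostly on one axis, right chains widely on both, while root-to-calf and knee joints rotate freely.
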PhysicsAsset /BRCosmetics/Characters/Player/Male/Medium/Bodies/M_MED_Cyclone/Meshes/M_MED_Cyclone_Physics
bodies {
 calf_l [{shapeType=Capsule, length=35.3403, radius=6.2969, center=<-18.7353 -2.62 0.0003>, rotation=<rot 90 3.7927 3.7927>}]
 calf_r [
  {shapeType=Capsule, length=36.0903, radius=5.5469, center=<18.7353 2.62 -0.0003>, rotation=<rot 90 -5.0511 174.9489>}
  {shapeType=Capsule, length=10.75, radius=5.25, center=<10.0735 3.0452 2.3715>, rotation=<rot 90 -0.0198 -0.0199>}
 ]
 dyn_legdangle_bt_01_l [{shapeType=Capsule, length=5.3115, radius=1.5, center=<3.2156 -0.0138 0.0658>, rotation=<rot 0 -90.2328 -88.3008>}]
 dyn_legdangle_bt_01_r [{shapeType=Capsule, length=4.8115, radius=2, center=<3.2156 -0.0138 0.0658>, rotation=<rot 0 -90.2328 -88.3008>}]
 dyn_legdangle_bt_02_l [{shapeType=Capsule, length=5.3115, radius=1.5, center=<3.2156 -0.0138 0.0658>, rotation=<rot 0 -90.2328 -88.3008>}]
 dyn_legdangle_bt_02_r [{shapeType=Capsule, length=4.8115, radius=2, center=<3.2156 -0.0138 0.0658>, rotation=<rot 0 -90.2328 -88.3008>}]
 dyn_legdangle_l [{shapeType=Box, extentX=0.01, extentY=0.01, extentZ=0.01}]
 dyn_legdangle_r [{shapeType=Box, extentX=0.01, extentY=0.01, extentZ=0.01}]
 dyn_legdangle_tp_01_l [{shapeType=Capsule, length=5.3115, radius=1.5, center=<3.2156 -0.0138 0.0658>, rotation=<rot 0 -90.2328 -88.3008>}]
 dyn_legdangle_tp_01_r [{shapeType=Capsule, length=5.3115, radius=1.5, center=<3.2156 -0.0138 0.0658>, rotation=<rot 0 -90.2328 -88.3008>}]
 dyn_legdangle_tp_02_l [{shapeType=Capsule, length=5.3115, radius=1.5, center=<3.2156 -0.0138 0.0658>, rotation=<rot 0 -90.2328 -88.3008>}]
 dyn_legdangle_tp_02_r [{shapeType=Capsule, length=5.3115, radius=1.5, center=<3.2156 -0.0138 0.0658>, rotation=<rot 0 -90.2328 -88.3008>}]
 thigh_l [{shapeType=Capsule, length=35.3403, radius=6.2969, center=<-18.7353 -2.62 0.0003>, rotation=<rot 90 3.7927 3.7927>}]
 thigh_r [
  {shapeType=Capsule, length=33.3403, radius=8.2969, center=<18.7353 2.62 -0.0003>, rotation=<rot 90 -5.0511 174.9489>}
  {shapeType=Capsule, length=7.0903, radius=6.7969, center=<22.7984 2.6201 7.3332>, rotation=<rot 90 -0.0033 179.9967>}
 ]
}
constraints [
 {bone1Name=dyn_legdangle_tp_01_l, bone2Name=dyn_legdangle_l, swing1LimitDegrees=75, swing2LimitDegrees=5, twistLimitDegrees=5}
 {bone1Name=dyn_legdangle_tp_02_l, bone2Name=dyn_legdangle_tp_01_l, swing1LimitDegrees=75, swing2LimitDegrees=5, twistLimitDegrees=5}
 {bone1Name=dyn_legdangle_bt_01_l, bone2Name=dyn_legdangle_l, swing1LimitDegrees=75, swing2LimitDegrees=5, twistLimitDegrees=5}
 {bone1Name=dyn_legdangle_bt_02_l, bone2Name=dyn_legdangle_bt_01_l, swing1LimitDegrees=75, swing2LimitDegrees=5, twistLimitDegrees=5}
 {bone1Name=dyn_legdangle_bt_01_r, bone2Name=dyn_legdangle_r, swing1LimitDegrees=75, swing2LimitDegrees=75, twistLimitDegrees=5}
 {bone1Name=dyn_legdangle_bt_02_r, bone2Name=dyn_legdangle_bt_01_r, swing1LimitDegrees=75, swing2LimitDegrees=75, twistLimitDegrees=5}
 {bone1Name=dyn_legdangle_tp_01_r, bone2Name=dyn_legdangle_r, swing1LimitDegrees=75, swing2LimitDegrees=75, twistLimitDegrees=5}
 {bone1Name=dyn_legdangle_tp_02_r, bone2Name=dyn_legdangle_tp_01_r, swing1LimitDegrees=75, swing2LimitDegrees=75, twistLimitDegrees=5}
 {
  bone1Name=dyn_legdangle_l
  bone2Name=calf_l
  swing1LimitDegrees=45
  swing1Motion=Free
  swing2LimitDegrees=45
  swing2Motion=Free
  twistLimitDegrees=45
  twistMotion=Free
 }
 {
  bone1Name=dyn_legdangle_r
  bone2Name=calf_r
  swing1LimitDegrees=45
  swing1Motion=Free
  swing2LimitDegrees=45
  swing2Motion=Free
  twistLimitDegrees=45
  twistMotion=Free
 }
 {
  bone1Name=calf_l
  bone2Name=thigh_l
  swing1LimitDegrees=45
  swing1Motion=Free
  swing2LimitDegrees=45
  swing2Motion=Free
  twistLimitDegrees=45
  twistMotion=Free
 }
 {
  bone1Name=calf_r
  bone2Name=thigh_r
  swing1LimitDegrees=45
  swing1Motion=Free
  swing2LimitDegrees=45
  swing2Motion=Free
  twistLimitDegrees=45
  twistMotion=Free
 }
]
