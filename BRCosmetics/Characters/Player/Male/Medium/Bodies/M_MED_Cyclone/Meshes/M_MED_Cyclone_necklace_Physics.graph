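# Physics asset for the necklace on the male medium Cyclone player cosmetic, defining capsule bodies and joints for its dangling front and back pieces.
# Four tiny MSTR anchors on spine_04 root two front chains and two three-way back sets, each pair being a plain and "tp" variant.
# Necklace joints have limited swing and mostly tight twist.
# Spine, clavicle, pectoral and upper-arm capsules approximate the torso, linked by angularly free constraints.
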PhysicsAsset /BRCosmetics/Characters/Player/Male/Medium/Bodies/M_MED_Cyclone/Meshes/M_MED_Cyclone_necklace_Physics
bodies {
 clavicle_l [{shapeType=Capsule, length=11.2659, radius=6.8678, center=<8.3479 -2.7274 0>, rotation=<rot 90 0.1055 -179.8945>}]
 clavicle_r [{shapeType=Capsule, length=12.7659, radius=5.3678, center=<-8.3478 -0.0001 0>, rotation=<rot 90 -179.7626 -179.7626>}]
 dyn_necklaceBack_01 [{shapeType=Capsule, length=8.5, radius=2.5, shapeName=dyn_necklaceBack_01_capsule, center=<5.5 0 0>, rotation=<rot 0 0 90>}]
 dyn_necklaceBack_01_MSTR [{shapeType=Capsule, length=0.505, radius=0.505, shapeName=dyn_necklaceBack_01_MSTR_capsule, rotation=<rot 0 -90 0>}]
 dyn_necklaceBack_01_l [{shapeType=Capsule, length=11.5, radius=2.5, shapeName=dyn_necklaceBack_01_l_capsule, center=<6.5 0 0>, rotation=<rot 90 0 0>}]
 dyn_necklaceBack_01_r [{shapeType=Capsule, length=11.5, radius=2.5, shapeName=dyn_necklaceBack_01_r_capsule, center=<6.55 0 0>, rotation=<rot 90 0 0>}]
 dyn_necklaceBack_tp_01 [{shapeType=Capsule, length=8.5, radius=2.5, shapeName=dyn_necklaceBack_tp_01_capsule, center=<5.5 0 0>, rotation=<rot 0 0 90>}]
 dyn_necklaceBack_tp_01_MSTR [{shapeType=Capsule, length=0.505, radius=0.505, shapeName=dyn_necklaceBack_tp_01_MSTR_capsule, rotation=<rot 0 -90 0>}]
 dyn_necklaceBack_tp_01_l [{shapeType=Capsule, length=11.5, radius=2.5, shapeName=dyn_necklaceBack_tp_01_l_capsule, center=<6.5 0 0>, rotation=<rot 90 0 0>}]
 dyn_necklaceBack_tp_01_r [{shapeType=Capsule, length=11.5, radius=2.5, shapeName=dyn_necklaceBack_tp_01_r_capsule, center=<6.5 0 0>, rotation=<rot 90 0 0>}]
 dyn_necklace_01 [{shapeType=Capsule, length=22, radius=2, shapeName=dyn_necklace_01_capsule, center=<3 0 2>, rotation=<rot 0 0 90>}]
 dyn_necklace_01_MSTR [{shapeType=Capsule, length=0.505, radius=0.505, shapeName=dyn_necklace_01_MSTR_capsule, rotation=<rot 0 -90 0>}]
 dyn_necklace_02 [{shapeType=Capsule, length=17, radius=1, shapeName=dyn_necklace_tp_02_capsule, center=<0 0 1>, rotation=<rot 0 0 90>}]
 dyn_necklace_03 [{shapeType=Capsule, length=12, radius=1, shapeName=dyn_necklace_03_capsule, center=<0 0 1>, rotation=<rot 0 0 90>}]
 dyn_necklace_04 [{shapeType=Capsule, length=6.26, radius=0.74, shapeName=dyn_necklace_04_capsule, center=<0 0 1>, rotation=<rot 0 0 90>}]
 dyn_necklace_tp_01 [{shapeType=Capsule, length=22, radius=2.5, shapeName=dyn_necklace_tp_01_capsule, rotation=<rot 0 0 90>}]
 dyn_necklace_tp_01_MSTR [{shapeType=Capsule, length=0.01, radius=0.01}]
 dyn_necklace_tp_02 [{shapeType=Capsule, length=17, radius=1, shapeName=dyn_necklace_tp_02_capsule, center=<2 0 0.25>, rotation=<rot 0 0 90>}]
 dyn_necklace_tp_03 [{shapeType=Capsule, length=12.1079, radius=0.8921, shapeName=dyn_necklace_tp_03_capsule, center=<2 0 0.4443>, rotation=<rot 0 0 90>}]
 dyn_necklace_tp_04 [{shapeType=Capsule, length=5.9564, radius=1.0436, shapeName=dyn_necklace_tp_04_capsule, center=<2.0817 0 -0.144>, rotation=<rot 0 0 90>}]
 dyn_necklace_tp_05 [
  {shapeType=Capsule, length=6.413, radius=1.0869, shapeName=dyn_necklace_tp_05_capsule, center=<3.9807 0 0.25>, rotation=<rot 90 0 0>}
  {shapeType=Capsule, length=2.6924, radius=1.0789, shapeName=dyn_necklace_tp_05_capsule, center=<2.9884 0 0.0119>, rotation=<rot 0.0001 0 -90>}
 ]
 pectoralis_01_l [{shapeType=Capsule, length=12.5159, radius=5.6178, center=<10.5674 -0.0003 -3.1706>, rotation=<rot 90 -0.633 179.367>}]
 pectoralis_01_r [{shapeType=Capsule, length=12.5159, radius=5.6178, center=<10.5674 -0.0003 -3.1706>, rotation=<rot 90 -0.633 179.367>}]
 spine_03 [
  {shapeType=Capsule, length=17.518, radius=11.1852, center=<7.5561 1.1 0.4779>, rotation=<rot -4.8634 -90 -0.5327>}
  {shapeType=Capsule, length=20.25, radius=9.75, center=<7.9945 -1.2228 0>}
 ]
 spine_04 [
  {shapeType=Capsule, length=16.4764, radius=12.2267, center=<4.8071 -0.4043 0.3244>, rotation=<rot -4.8634 -90 -0.5327>}
  {shapeType=Capsule, length=20.25, radius=9.75, center=<7.9945 -1.2228 0>}
 ]
 upperarm_l [{shapeType=Capsule, length=24.2659, radius=5.3678, center=<11.7662 0 0>, rotation=<rot 90 -174.9489 5.0512>}]
 upperarm_r [{shapeType=Capsule, length=24.2659, radius=5.3678, center=<-11.7662 0 0>, rotation=<rot 90 7.5524 7.5525>}]
}
constraints [
 {
  bone1Name=dyn_necklaceBack_01_MSTR
  bone2Name=spine_04
  swing1LimitDegrees=45
  swing1Motion=Free
  swing2LimitDegrees=45
  swing2Motion=Free
  twistLimitDegrees=45
  twistMotion=Free
 }
 {
  bone1Name=dyn_necklace_01_MSTR
  bone2Name=spine_04
  swing1LimitDegrees=45
  swing1Motion=Free
  swing2LimitDegrees=45
  swing2Motion=Free
  twistLimitDegrees=45
  twistMotion=Free
 }
 {
  bone1Name=dyn_necklace_tp_01_MSTR
  bone2Name=spine_04
  swing1LimitDegrees=45
  swing1Motion=Free
  swing2LimitDegrees=45
  swing2Motion=Free
  twistLimitDegrees=45
  twistMotion=Free
 }
 {
  bone1Name=dyn_necklaceBack_tp_01_MSTR
  bone2Name=spine_04
  swing1LimitDegrees=45
  swing1Motion=Free
  swing2LimitDegrees=45
  swing2Motion=Free
  twistLimitDegrees=45
  twistMotion=Free
 }
 {bone1Name=dyn_necklaceBack_01, bone2Name=dyn_necklaceBack_01_MSTR, swing1LimitDegrees=1, swing2LimitDegrees=40, twistLimitDegrees=1}
 {bone1Name=dyn_necklaceBack_01_l, bone2Name=dyn_necklaceBack_01_MSTR, swing1LimitDegrees=1, swing2LimitDegrees=30, twistLimitDegrees=1}
 {bone1Name=dyn_necklace_01, bone2Name=dyn_necklace_01_MSTR, swing1LimitDegrees=10, swing2LimitDegrees=45, twistLimitDegrees=1}
 {bone1Name=dyn_necklace_02, bone2Name=dyn_necklace_01, swing1LimitDegrees=20, swing2LimitDegrees=45, twistLimitDegrees=20}
 {bone1Name=dyn_necklace_03, bone2Name=dyn_necklace_02, swing1LimitDegrees=20, swing2LimitDegrees=45, twistLimitDegrees=20}
 {bone1Name=dyn_necklace_tp_01, bone2Name=dyn_necklace_tp_01_MSTR, swing1LimitDegrees=20, swing2LimitDegrees=20, twistLimitDegrees=1}
 {bone1Name=dyn_necklace_tp_02, bone2Name=dyn_necklace_tp_01, swing1LimitDegrees=10, swing2LimitDegrees=20, twistLimitDegrees=1}
 {bone1Name=dyn_necklace_tp_03, bone2Name=dyn_necklace_tp_02, swing1LimitDegrees=10, swing2LimitDegrees=20, twistLimitDegrees=1}
 {bone1Name=dyn_necklace_tp_04, bone2Name=dyn_necklace_tp_03, swing1LimitDegrees=10, swing2LimitDegrees=20, twistLimitDegrees=1}
 {bone1Name=dyn_necklace_tp_05, bone2Name=dyn_necklace_tp_04, swing1LimitDegrees=45, swing2LimitDegrees=45, twistLimitDegrees=5}
 {bone1Name=dyn_necklaceBack_tp_01, bone2Name=dyn_necklaceBack_tp_01_MSTR, swing1LimitDegrees=1, swing2LimitDegrees=30, twistLimitDegrees=1}
 {bone1Name=dyn_necklaceBack_tp_01_l, bone2Name=dyn_necklaceBack_tp_01_MSTR, swing1LimitDegrees=1, swing2LimitDegrees=30, twistLimitDegrees=1}
 {bone1Name=dyn_necklaceBack_01_r, bone2Name=dyn_necklaceBack_01_MSTR, swing1LimitDegrees=1, swing2LimitDegrees=30, twistLimitDegrees=1}
 {bone1Name=dyn_necklaceBack_tp_01_r, bone2Name=dyn_necklaceBack_tp_01_MSTR, swing1LimitDegrees=1, swing2LimitDegrees=30, twistLimitDegrees=1}
 {
  bone1Name=upperarm_r
  bone2Name=clavicle_r
  swing1LimitDegrees=45
  swing1Motion=Free
  swing2LimitDegrees=45
  swing2Motion=Free
  twistLimitDegrees=45
  twistMotion=Free
 }
 {
  bone1Name=upperarm_l
  bone2Name=clavicle_l
  swing1LimitDegrees=45
  swing1Motion=Free
  swing2LimitDegrees=45
  swing2Motion=Free
  twistLimitDegrees=45
  twistMotion=Free
 }
 {
  bone1Name=clavicle_l
  bone2Name=spine_04
  swing1LimitDegrees=45
  swing1Motion=Free
  swing2LimitDegrees=45
  swing2Motion=Free
  twistLimitDegrees=45
  twistMotion=Free
 }
 {
  bone1Name=clavicle_r
  bone2Name=spine_04
  swing1LimitDegrees=45
  swing1Motion=Free
  swing2LimitDegrees=45
  swing2Motion=Free
  twistLimitDegrees=45
  twistMotion=Free
 }
 {
  bone1Name=pectoralis_01_l
  bone2Name=spine_04
  swing1LimitDegrees=45
  swing1Motion=Free
  swing2LimitDegrees=45
  swing2Motion=Free
  twistLimitDegrees=45
  twistMotion=Free
 }
 {
  bone1Name=pectoralis_01_r
  bone2Name=spine_04
  swing1LimitDegrees=45
  swing1Motion=Free
  swing2LimitDegrees=45
  swing2Motion=Free
  twistLimitDegrees=45
  twistMotion=Free
 }
 {bone1Name=dyn_necklace_04, bone2Name=dyn_necklace_03, swing1LimitDegrees=20, swing2LimitDegrees=45, twistLimitDegrees=20}
 {
  bone1Name=spine_04
  bone2Name=spine_03
  swing1LimitDegrees=45
  swing1Motion=Free
  swing2LimitDegrees=45
  swing2Motion=Free
  twistLimitDegrees=45
  twistMotion=Free
 }
]
