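# Post-process-style AnimBlueprint for the Cyclone male cosmetic body, adding secondary motion and muscle correctives to the pose copied from its attached parent.
# The full chain applies look-at, constraint and arm-driven shoulder/bicep/tricep correctives, then body and necklace RigidBody sims using skydiving-driven gravity.
# When rigid bodies are disabled or LOD exceeds 1, only leg-dangle trail controllers run.
# Crouch, skydive and Cyclone emote alphas drive extra bone offsets.
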AnimBlueprint /BRCosmetics/Characters/Player/Male/Medium/Bodies/M_MED_Cyclone/Meshes/M_MED_Cyclone_AnimBP
target_skeleton=&/M_MED_Cyclone_Skeleton
graphs [
 AnimGraph {nodes [
   n1 = "Transform (Modify) Bone - Bone: dyn_legdangle_l"(
    ComponentPose=Local To Component(LocalPose=Use cached pose 'initCache'())
    Alpha=1
    Translation="0,0,0"
    Rotation="0.000,-70,0.000"
    Scale="1,1,1"
   ) {
    rotationMode=BMM_Additive
    rotationSpace=BCS_BoneSpace
    alphaBoolBlend {blendOption=Linear}
    boneToModify {boneName=dyn_legdangle_l}
    rotation {yaw=-70}
   }
   n2 = Trail controller - Bone: dyn_legdangle_tp_02_l!(ComponentPose=@n1, Alpha=1, RelaxationSpeedScale:float=1.5, FakeVelocity:Vector="0,0,0") -> (Pose) {
    bActorSpaceFakeVel=true
    bLimitRotation=true
    bUsePlanarLimit=true
    debugLifeTime=0.1
    maxDeltaTime=0.0333
    relaxationSpeedScale=1.2
    alphaScaleBiasClamp {bInterpResult=true, interpSpeedDecreasing=25, interpSpeedIncreasing=25, scale=1.5}
    baseJoint {boneName=root}
    planarLimits [
     {drivingBone {boneName=root}, planeTransform {location=<0 0 10>, rotation=<rot 0 0 0>, scale=<1 1 1>}}
     {drivingBone {boneName=spine_04}, planeTransform {location=<0 40 0>, rotation=<rot 0 0 -90>, scale=<1 1 1>}}
    ]
    rotationLimits [{limitMax=<0 0 30>, limitMin=<0 -30 -30>}, {limitMax=<0 0 30>, limitMin=<0 -30 -30>}]
    rotationOffsets [<0 0 0>, <0 0 0>]
    trailBone {boneName=dyn_legdangle_tp_02_l}
    trailRelaxationSpeed {editorCurveData {keys [<at 0 15 RCIM_Linear>,<at 0.5289 13.1838 RCIM_Linear>,<at 1 6 RCIM_Linear>]}}
   }
   n3 = Trail controller - Bone: dyn_legdangle_bt_02_l!(ComponentPose=@n2, Alpha=1, RelaxationSpeedScale:float=1.5, FakeVelocity:Vector="0,0,0") -> (Pose) {
    bActorSpaceFakeVel=true
    bLimitRotation=true
    bUsePlanarLimit=true
    debugLifeTime=0.1
    maxDeltaTime=0.0333
    relaxationSpeedScale=1.2
    alphaScaleBiasClamp {bInterpResult=true, interpSpeedDecreasing=25, interpSpeedIncreasing=25, scale=1.5}
    baseJoint {boneName=root}
    planarLimits [
     {drivingBone {boneName=root}, planeTransform {location=<0 0 10>, rotation=<rot 0 0 0>, scale=<1 1 1>}}
     {drivingBone {boneName=spine_04}, planeTransform {location=<0 40 0>, rotation=<rot 0 0 -90>, scale=<1 1 1>}}
    ]
    rotationLimits [{limitMax=<0 0 30>, limitMin=<0 -30 -30>}, {limitMax=<0 0 30>, limitMin=<0 -30 -30>}]
    rotationOffsets [<0 0 0>, <0 0 0>]
    trailBone {boneName=dyn_legdangle_bt_02_l}
    trailRelaxationSpeed {editorCurveData {keys [<at 0 15 RCIM_Linear>,<at 0.5289 13.1838 RCIM_Linear>,<at 1 6 RCIM_Linear>]}}
   }
   n4 = "Transform (Modify) Bone - Bone: dyn_legdangle_r"(ComponentPose=@n3, Alpha=1, Translation="0,0,0", Rotation="0.000,70,0.000", Scale="1,1,1") {
    rotationMode=BMM_Additive
    rotationSpace=BCS_BoneSpace
    alphaBoolBlend {blendOption=Linear}
    boneToModify {boneName=dyn_legdangle_r}
    rotation {yaw=70}
   }
   n5 = Trail controller - Bone: dyn_legdangle_tp_02_r!(ComponentPose=@n4, Alpha=1, RelaxationSpeedScale:float=1.5, FakeVelocity:Vector="0,0,0") -> (Pose) {
    bActorSpaceFakeVel=true
    bLimitRotation=true
    bUsePlanarLimit=true
    debugLifeTime=0.1
    maxDeltaTime=0.0333
    relaxationSpeedScale=1.2
    alphaScaleBiasClamp {bInterpResult=true, interpSpeedDecreasing=25, interpSpeedIncreasing=25, scale=1.5}
    baseJoint {boneName=root}
    planarLimits [
     {drivingBone {boneName=root}, planeTransform {location=<0 0 10>, rotation=<rot 0 0 0>, scale=<1 1 1>}}
     {drivingBone {boneName=spine_04}, planeTransform {location=<0 40 0>, rotation=<rot 0 0 -90>, scale=<1 1 1>}}
    ]
    rotationLimits [{limitMax=<0 0 30>, limitMin=<0 -30 -30>}, {limitMax=<0 0 30>, limitMin=<0 -30 -30>}]
    rotationOffsets [<0 0 0>, <0 0 0>]
    trailBone {boneName=dyn_legdangle_tp_02_r}
    trailRelaxationSpeed {editorCurveData {keys [<at 0 15 RCIM_Linear>,<at 0.5289 13.1838 RCIM_Linear>,<at 1 6 RCIM_Linear>]}}
   }
   n6 = Trail controller - Bone: dyn_legdangle_bt_02_r!(ComponentPose=@n5, Alpha=1, RelaxationSpeedScale:float=1.5, FakeVelocity:Vector="0,0,0") -> (Pose) {
    bActorSpaceFakeVel=true
    bLimitRotation=true
    bUsePlanarLimit=true
    debugLifeTime=0.1
    maxDeltaTime=0.0333
    relaxationSpeedScale=1.2
    alphaScaleBiasClamp {bInterpResult=true, interpSpeedDecreasing=25, interpSpeedIncreasing=25, scale=1.5}
    baseJoint {boneName=root}
    planarLimits [
     {drivingBone {boneName=root}, planeTransform {location=<0 0 10>, rotation=<rot 0 0 0>, scale=<1 1 1>}}
     {drivingBone {boneName=spine_04}, planeTransform {location=<0 40 0>, rotation=<rot 0 0 -90>, scale=<1 1 1>}}
    ]
    rotationLimits [{limitMax=<0 0 30>, limitMin=<0 -30 -30>}, {limitMax=<0 0 30>, limitMin=<0 -30 -30>}]
    rotationOffsets [<0 0 0>, <0 0 0>]
    trailBone {boneName=dyn_legdangle_bt_02_r}
    trailRelaxationSpeed {editorCurveData {keys [<at 0 15 RCIM_Linear>,<at 0.5289 13.1838 RCIM_Linear>,<at 1 6 RCIM_Linear>]}}
   }
   n7 = "Constraint - scapula_base_l (P)"(ComponentPose=Local To Component(LocalPose=Use cached pose 'initCache'()), Alpha=1, ConstraintWeights_0=1) {
    alphaBoolBlend {blendOption=Linear}
    boneToModify {boneName=scapula_base_l}
    constraintSetup [{targetBone {boneName=clavicle_l}}]
    constraintWeights [1]
   }
   n8 = Look At - Bone: scapula_00_l(ComponentPose=@n7, Alpha=1) {
    alphaBoolBlend {blendOption=Linear}
    boneToModify {boneName=scapula_00_l}
    lookAtLocation {x=0}
    lookAtTarget {boneReference {boneName=scapula_01_l_aim}}
    lookAt_Axis {axis=<1 0>}
    lookUp_Axis {axis {x=0, z=1}}
   }
   n9 = Look At - Bone: pectoralis_01_l(ComponentPose=@n8, Alpha=1) {
    alphaBoolBlend {blendOption=Linear}
    boneToModify {boneName=pectoralis_01_l}
    lookAtLocation {x=0}
    lookAtTarget {boneReference {boneName=upperarm_l}}
    lookAt_Axis {axis=<1 0>}
    lookUp_Axis {axis {x=0, z=1}}
   }
   n10 = Look At - Bone: lattissimus_00_l(ComponentPose=@n9, Alpha=1) {
    alphaBoolBlend {blendOption=Linear}
    boneToModify {boneName=lattissimus_00_l}
    lookAtLocation {x=0}
    lookAtTarget {boneReference {boneName=upperarm_l}}
    lookAt_Axis {axis=<1 0>}
    lookUp_Axis {axis {x=0, z=1}}
   }
   n11 = "Constraint - scapula_base_r (P)"(ComponentPose=@n10, Alpha=1, ConstraintWeights_0=1) {
    alphaBoolBlend {blendOption=Linear}
    boneToModify {boneName=scapula_base_r}
    constraintSetup [{targetBone {boneName=clavicle_r}}]
    constraintWeights [1]
   }
   n12 = Look At - Bone: scapula_00_r(ComponentPose=@n11, Alpha=1) {
    alphaBoolBlend {blendOption=Linear}
    boneToModify {boneName=scapula_00_r}
    lookAtLocation {x=0}
    lookAtTarget {boneReference {boneName=scapula_01_r_aim}}
    lookAt_Axis {axis=<1 0>}
    lookUp_Axis {axis {x=0, z=1}}
   }
   n13 = Look At - Bone: pectoralis_01_r(ComponentPose=@n12, Alpha=1) {
    alphaBoolBlend {blendOption=Linear}
    boneToModify {boneName=pectoralis_01_r}
    lookAtLocation {x=0}
    lookAtTarget {boneReference {boneName=upperarm_r}}
    lookAt_Axis {axis=<1 0>}
    lookUp_Axis {axis {x=0, z=1}}
   }
   n14 = Look At - Bone: lattissimus_00_r(ComponentPose=@n13, Alpha=1) {
    alphaBoolBlend {blendOption=Linear}
    boneToModify {boneName=lattissimus_00_r}
    lookAtLocation {x=0}
    lookAtTarget {boneReference {boneName=upperarm_r}}
    lookAt_Axis {axis=<1 0>}
    lookUp_Axis {axis {x=0, z=1}}
   }
   n15 = "Transform (Modify) Bone - Bone: upperarm_twist_01_r"(ComponentPose=@n14, Alpha=Get RscaleShoulder() -> (RscaleShoulder:double), Translation="0,0,0", Rotation="0,0,0", Scale="1.2,1.2,1.2") {
    rotationMode=BMM_Additive
    rotationSpace=BCS_BoneSpace
    scaleMode=BMM_Additive
    scaleSpace=BCS_BoneSpace
    translationMode=BMM_Additive
    translationSpace=BCS_BoneSpace
    alphaBoolBlend {blendOption=Linear}
    boneToModify {boneName=upperarm_twist_01_r}
    scale=<1.2 1.2 1.2>
   }
   n16 = "Transform (Modify) Bone - Bone: upperarm_twist_01_l"(ComponentPose=@n15, Alpha=Get LscaleShoulder() -> (LscaleShoulder:double), Translation="0,0,0", Rotation="0,0,0", Scale="1.2,1.2,1.2") {
    rotationMode=BMM_Additive
    rotationSpace=BCS_BoneSpace
    scaleMode=BMM_Additive
    scaleSpace=BCS_BoneSpace
    translationMode=BMM_Additive
    translationSpace=BCS_BoneSpace
    alphaBoolBlend {blendOption=Linear}
    boneToModify {boneName=upperarm_twist_01_l}
    scale=<1.2 1.2 1.2>
   }
   n17 = Get LarmDistanceBICEPAlpha() -> (LarmDistanceBICEPAlpha:double)
   n18 = "Transform (Modify) Bone - Bone: bicep_02_l"(ComponentPose=@n16, Alpha=@n17, Translation="-6,-0.000,-0.000", Rotation="0,0,0", Scale="1,1,1") {
    rotationMode=BMM_Additive
    rotationSpace=BCS_BoneSpace
    scaleMode=BMM_Additive
    scaleSpace=BCS_BoneSpace
    translationMode=BMM_Additive
    translationSpace=BCS_BoneSpace
    alphaBoolBlend {blendOption=Linear}
    boneToModify {boneName=bicep_02_l}
    translation {x=-6}
   }
   n19 = "Transform (Modify) Bone - Bone: bicep_01_l"(ComponentPose=@n18, Alpha=@n17, Translation="-0.000,0.000,2.343", Rotation="0,0,0", Scale="1.000,1.500,1.250") {
    rotationMode=BMM_Additive
    rotationSpace=BCS_BoneSpace
    scaleMode=BMM_Additive
    scaleSpace=BCS_BoneSpace
    translationMode=BMM_Additive
    translationSpace=BCS_BoneSpace
    alphaBoolBlend {blendOption=Linear}
    boneToModify {boneName=bicep_01_l}
    scale {y=1.5, z=1.25}
    translation {z=2.343}
   }
   n20 = "Transform (Modify) Bone - Bone: tricep_02_l"(ComponentPose=@n19, Alpha=@n17, Translation="-6,-0.000,-0.000", Rotation="0,0,0", Scale="1,1,1") {
    rotationMode=BMM_Additive
    rotationSpace=BCS_BoneSpace
    scaleMode=BMM_Additive
    scaleSpace=BCS_BoneSpace
    translationMode=BMM_Additive
    translationSpace=BCS_BoneSpace
    alphaBoolBlend {blendOption=Linear}
    boneToModify {boneName=tricep_02_l}
    translation {x=-6}
   }
   n21 = "Transform (Modify) Bone - Bone: tricep_01_l"(ComponentPose=@n20, Alpha=@n17, Translation="-0.000,0.000,2.343", Rotation="0,0,0", Scale="1.000,1.500,1.250") {
    rotationMode=BMM_Additive
    rotationSpace=BCS_BoneSpace
    scaleMode=BMM_Additive
    scaleSpace=BCS_BoneSpace
    translationMode=BMM_Additive
    translationSpace=BCS_BoneSpace
    alphaBoolBlend {blendOption=Linear}
    alphaScaleBiasClamp {scale=0.3}
    boneToModify {boneName=tricep_01_l}
    scale {y=1.5, z=1.25}
    translation {z=2.343}
   }
   n22 = Get RarmDistanceBICEPAlpha() -> (RarmDistanceBICEPAlpha:double)
   n23 = "Transform (Modify) Bone - Bone: bicep_02_r"(ComponentPose=@n21, Alpha=@n22, Translation="-6,-0.000,-0.000", Rotation="0,0,0", Scale="1,1,1") {
    rotationMode=BMM_Additive
    rotationSpace=BCS_BoneSpace
    scaleMode=BMM_Additive
    scaleSpace=BCS_BoneSpace
    translationMode=BMM_Additive
    translationSpace=BCS_BoneSpace
    alphaBoolBlend {blendOption=Linear}
    boneToModify {boneName=bicep_02_r}
    translation {x=-6}
   }
   n24 = "Transform (Modify) Bone - Bone: bicep_01_r"(ComponentPose=@n23, Alpha=@n22, Translation="-0.000,0.000,2.343", Rotation="0,0,0", Scale="1.000,1.500,1.250") {
    rotationMode=BMM_Additive
    rotationSpace=BCS_BoneSpace
    scaleMode=BMM_Additive
    scaleSpace=BCS_BoneSpace
    translationMode=BMM_Additive
    translationSpace=BCS_BoneSpace
    alphaBoolBlend {blendOption=Linear}
    boneToModify {boneName=bicep_01_r}
    scale {y=1.5, z=1.25}
    translation {z=2.343}
   }
   n25 = "Transform (Modify) Bone - Bone: tricep_02_r"(ComponentPose=@n24, Alpha=@n22, Translation="-6,-0.000,-0.000", Rotation="0,0,0", Scale="1,1,1") {
    rotationMode=BMM_Additive
    rotationSpace=BCS_BoneSpace
    scaleMode=BMM_Additive
    scaleSpace=BCS_BoneSpace
    translationMode=BMM_Additive
    translationSpace=BCS_BoneSpace
    alphaBoolBlend {blendOption=Linear}
    alphaScaleBiasClamp {scale=0.5}
    boneToModify {boneName=tricep_02_r}
    translation {x=-6}
   }
   n26 = "Transform (Modify) Bone - Bone: tricep_01_r"(ComponentPose=@n25, Alpha=@n22, Translation="-0.000,0.000,2.343", Rotation="0,0,0", Scale="1.000,1.500,1.250") {
    rotationMode=BMM_Additive
    rotationSpace=BCS_BoneSpace
    scaleMode=BMM_Additive
    scaleSpace=BCS_BoneSpace
    translationMode=BMM_Additive
    translationSpace=BCS_BoneSpace
    alphaBoolBlend {blendOption=Linear}
    alphaScaleBiasClamp {scale=0.3}
    boneToModify {boneName=tricep_01_r}
    scale {y=1.5, z=1.25}
    translation {z=2.343}
   }
   n27 = "Constraint - dyn_necklaceBack_tp_01_l (R)"(ComponentPose=@n26, Alpha=1, ConstraintWeights_0=1) {
    alphaBoolBlend {blendOption=Linear}
    boneToModify {boneName=dyn_necklaceBack_tp_01_l}
    constraintSetup [{transformType=Rotation, targetBone {boneName=clavicle_l}}]
    constraintWeights [1]
   }
   n28 = "Constraint - dyn_necklaceBack_01_l (R)"(ComponentPose=@n27, Alpha=1, ConstraintWeights_0=1) {
    alphaBoolBlend {blendOption=Linear}
    boneToModify {boneName=dyn_necklaceBack_01_l}
    constraintSetup [{transformType=Rotation, targetBone {boneName=clavicle_l}}]
    constraintWeights [1]
   }
   n29 = "Constraint - dyn_necklaceBack_01_r (R)"(ComponentPose=@n28, Alpha=1, ConstraintWeights_0=1) {
    alphaBoolBlend {blendOption=Linear}
    boneToModify {boneName=dyn_necklaceBack_01_r}
    constraintSetup [{transformType=Rotation, targetBone {boneName=clavicle_r}}]
    constraintWeights [1]
   }
   n30 = "Constraint - dyn_necklaceBack_tp_01_r (R)"(ComponentPose=@n29, Alpha=1, ConstraintWeights_0=1) {
    alphaBoolBlend {blendOption=Linear}
    boneToModify {boneName=dyn_necklaceBack_tp_01_r}
    constraintSetup [{transformType=Rotation, targetBone {boneName=clavicle_r}}]
    constraintWeights [1]
   }
   n31 = "Transform (Modify) Bone - Bone: dyn_necklace_01_MSTR"(ComponentPose=@n30, Alpha=1, Translation="0.000,-0.000,0.811", Rotation="0,0,0", Scale="1,1,1") {
    rotationMode=BMM_Additive
    rotationSpace=BCS_BoneSpace
    translationMode=BMM_Additive
    translationSpace=BCS_BoneSpace
    alphaBoolBlend {blendOption=Linear}
    boneToModify {boneName=dyn_necklace_01_MSTR}
    translation {z=0.811}
   }
   n32 = "Transform (Modify) Bone - Bone: dyn_legdangle_r"(
    ComponentPose=@n31
    Alpha=Get isCrouchingAlpha() -> (isCrouchingAlpha:double)
    Translation="0,0,0"
    Rotation="0.000,115,0.000"
    Scale="1,1,1"
   ) {
    rotationMode=BMM_Additive
    rotationSpace=BCS_BoneSpace
    alphaBoolBlend {blendOption=Linear}
    boneToModify {boneName=dyn_legdangle_r}
    rotation {yaw=115}
   }
   n33 = Get windVector() -> (windVector:Vector)
   n34 = RigidBody!(ComponentPose=@n32, Alpha, OverrideWorldGravity:Vector=@n33, ExternalForce) -> (Pose) {
    bClampLinearTranslationLimitToRefPose=true
    bOverrideWorldGravity=true
    overridePhysicsAsset=&/M_MED_Cyclone_Physics
    simulationSpace=BaseBoneSpace
    alphaBoolBlend {blendOption=Linear}
    baseBoneRef {boneName=root}
    componentAppliedLinearAccClamp=<3000 3000 3000>
    componentLinearAccScale=<0.5 0.5 0.5>
    componentLinearVelScale=<0.7 0.7 0.7>
    overrideWorldGravity {z=-980}
   }
   n35 = RigidBody(ComponentPose=@n34, OverrideWorldGravity=@n33) {
    bClampLinearTranslationLimitToRefPose=true
    bOverrideWorldGravity=true
    overridePhysicsAsset=&/M_MED_Cyclone_necklace_Physics
    simulationSpace=BaseBoneSpace
    alphaBoolBlend {blendOption=Linear}
    baseBoneRef {boneName=spine_04}
    componentAppliedLinearAccClamp=<3000 3000 3000>
    componentLinearAccScale=<0.3 0.3 0.3>
    componentLinearVelScale=<0.3 0.3 0.5>
    overrideWorldGravity {z=-980}
   }
   n36 = "Transform (Modify) Bone - Bone: dyn_necklace_03"(ComponentPose=@n35, Alpha=1, Translation="0,0,0", Rotation="8,0.000,0.000", Scale="1,1,1") {
    rotationMode=BMM_Additive
    rotationSpace=BCS_BoneSpace
    translationMode=BMM_Additive
    translationSpace=BCS_BoneSpace
    alphaBoolBlend {blendOption=Linear}
    boneToModify {boneName=dyn_necklace_03}
    rotation {pitch=8}
   }
   n37 = "Transform (Modify) Bone - Bone: dyn_necklace_03"(ComponentPose=@n36, Alpha=Get isSkydiveAlpha() -> (isSkydiveAlpha:double), Translation="0,0,0", Rotation="15,0.000,0.000", Scale="1,1,1") {
    rotationMode=BMM_Additive
    rotationSpace=BCS_BoneSpace
    translationMode=BMM_Additive
    translationSpace=BCS_BoneSpace
    alphaBoolBlend {blendOption=Linear}
    boneToModify {boneName=dyn_necklace_03}
    rotation {pitch=15}
   }
   n38 = "Transform (Modify) Bone - Bone: dyn_necklace_01_MSTR"(
    ComponentPose=@n37
    Alpha=Get isCycloneDancingAlpha() -> (isCycloneDancingAlpha:double)
    Translation="0,0,0"
    Rotation="2,0.000,0.000"
    Scale="1,1,1"
   ) {
    rotationMode=BMM_Additive
    rotationSpace=BCS_BoneSpace
    translationMode=BMM_Additive
    translationSpace=BCS_BoneSpace
    alphaBoolBlend {blendOption=Linear}
    boneToModify {boneName=dyn_necklace_01_MSTR}
    rotation {pitch=2}
   }
   Output Pose(Result=Blend Poses by bool(
     BlendPose_0=Component To Local(ComponentPose=@n6)
     BlendPose_1=Component To Local(ComponentPose=@n38)
     bActiveValue=Get rigids_off() -> (rigids_off:bool)
    ) {blendType=Linear}) {layerGroup=None}
   Save cached pose 'initCache'(Pose=Copy Pose From Mesh() {bUseAttachedParent=true}) {cacheName=initCache}
  ]}
 EventGraph {nodes [
   Event BlueprintUpdateAnimation()
   > n1 = Sequence!(execute) -> (then_0, then_1, then_2:Exec, then_3:Exec, then_4:Exec, then_5:Exec)
   n2 = Break Fort Anim Input Skydiving External Force(FortAnimInput_SkydivingExternalForce=Get SkydivingExternalForceInput()) -> (
    bUseSkydivingVectorForce=~
    SkydivingVectorForce=~
    HeadToPelvisDirection=~
    FloatingMultiplier=~
    FloatingAdditive=~
    DivingMultiplier=~
    DivingAdditive=~
    ParachutingMultiplier=~
    ParachutingAdditive=~
   )
   > Set windVector(windVector:Vector=SelectVector(A=vector * float(A=@n2.SkydivingVectorForce, B=0.15), B="0, 0,-980", bPickA=Get bIsSkyDiving())) -> (Output_Get:Vector)
   Set isSkydiveAlpha(execute=@n1.then_1, isSkydiveAlpha:double="To Float (Boolean)"(InBool=Get bIsSkyDiving())) -> (Output_Get:double)
   n3 = Get PawnSpeed2D()
   n4 = Get bIsCrouching()
   > Set isCrouchingAlpha(isCrouchingAlpha:double=SelectFloat(A=MapRangeClamped(Value=@n3, InRangeA=0, InRangeB=150, OutRangeA=1, OutRangeB=0), B=0, bPickA=@n4)) -> (Output_Get:double)
   n5 = GetPartSkeletalMeshComponent()
   n6 = GetAlphaFromReferenceTransformDelta(
    SkeletalMeshComponent=@n5
    BoneName=lowerarm_r
    BaseName=upperarm_r
    DeltaComponentType=QuaternionTwist
    InputRange=(X=0,Y=-90)
    OutputRange=(X=0,Y=0.7)
    TwistAxis="0,0,1"
   )
   Set RbicepAlpha(execute=@n1.then_2, RbicepAlpha:double=@n6) -> (Output_Get:double)
   > Set RflexAlpha(RflexAlpha:double=MapRangeClamped(Value=@n6, InRangeA=0, InRangeB=1, OutRangeB=0.2)) -> (Output_Get:double)
   n7 = GetAlphaFromReferenceTransformDelta(
    SkeletalMeshComponent=@n5
    BoneName=lowerarm_l
    BaseName=upperarm_l
    DeltaComponentType=QuaternionTwist
    InputRange=(X=0,Y=-90)
    OutputRange=(X=0,Y=0.7)
    TwistAxis="0,0,1"
   )
   > Set LbicepAlpha(LbicepAlpha:double=@n7) -> (Output_Get:double)
   > Set LflexAlpha(LflexAlpha:double=MapRangeClamped(Value=@n7, InRangeA=0, InRangeB=1, OutRangeB=0.2)) -> (Output_Get:double)
   > Set RscaleShoulder(RscaleShoulder:double=GetAlphaFromReferenceTransformDelta(
     SkeletalMeshComponent=@n5
     BoneName=upperarm_r
     BaseName=clavicle_r
     DeltaComponentType=QuaternionTwist
     InputRange=(X=-90,Y=-130)
     TwistAxis="0,0,1"
    )) -> (Output_Get:double)
   > Set LscaleShoulder(LscaleShoulder:double=GetAlphaFromReferenceTransformDelta(
     SkeletalMeshComponent=@n5
     BoneName=upperarm_l
     BaseName=clavicle_l
     DeltaComponentType=QuaternionTwist
     InputRange=(X=-90,Y=-130)
     TwistAxis="0,0,1"
    )) -> (Output_Get:double)
   n8 = GetPartSkeletalMeshComponent()
   n9 = TransformFromBoneSpace!(
    execute=@n1.then_3
    self:Skinned Mesh Component Object Reference=@n8
    BoneName:Name=upperarm_l
    InPosition:Vector="0, 0, 0"
    InRotation:Rotator="0, 0, 0"
   ) -> (then, OutPosition:Vector, OutRotation:Rotator)
   > n10 = TransformFromBoneSpace(self=@n8, BoneName=lowerarm_twist_02_l, InPosition="0, 0, 0", InRotation="0, 0, 0")
   > Set LarmDistanceBICEPAlpha(LarmDistanceBICEPAlpha:double=MapRangeClamped(Value="Distance (Vector)"!(V1:Vector=@n9, V2:Vector=@n10) -> (ReturnValue:double), InRangeA=27, InRangeB=34, OutRangeA=1, OutRangeB=0)) -> (Output_Get:double)
   > n11 = TransformFromBoneSpace(self=@n8, BoneName=upperarm_r, InPosition="0, 0, 0", InRotation="0, 0, 0")
   > n12 = TransformFromBoneSpace(self=@n8, BoneName=lowerarm_twist_02_r, InPosition="0, 0, 0", InRotation="0, 0, 0")
   > Set RarmDistanceBICEPAlpha(RarmDistanceBICEPAlpha:double=MapRangeClamped(Value="Distance (Vector)"(V1=@n11, V2=@n12), InRangeA=27, InRangeB=34, OutRangeA=1, OutRangeB=0)) -> (Output_Get:double)
   n13 = Get CurrentLOD()
   n14 = Get bRigidBodyEnabled()
   Set rigids_off(execute=@n1.then_4, rigids_off:bool=OR Boolean(A=NOT Boolean(A=@n14), B="integer > integer"(A=@n13, B=1) {is=title K2Node_CallFunction})) -> (Output_Get:bool)
   n15 = GetAnimInstance(self=GetOwnerSkeletalMeshComponent())
   Is Valid(exec=@n1.then_5, InputObject=@n15)
   > Set isCycloneDancingAlpha(isCycloneDancingAlpha:double="To Float (Boolean)"(InBool=Montage_IsActive!(
      self:Anim Instance Object Reference=@n15
      Montage:Anim Montage Object Reference=/BRCosmetics/Animation/Game/MainPlayer/Emotes/Cyclone/Cyclone_CMM_M
     ) -> (ReturnValue:bool))) -> (Output_Get:double)
   TryGetPawnOwner()
   "To Float (Boolean)"(InBool=@n4)
  ]}
]
variables [
 isCrouchingAlpha {type=number}
 isCycloneDancingAlpha {type=number}
 isSkydiveAlpha {type=number}
 LarmDistanceBICEPAlpha {type=number}
 LbicepAlpha {type=number}
 LflexAlpha {type=number}
 LscaleShoulder {type=number}
 RarmDistanceBICEPAlpha {type=number}
 RbicepAlpha {type=number}
 RflexAlpha {type=number}
 rigids_off {type=boolean}
 RscaleShoulder {type=number}
 windVector {type=Vector}
]
