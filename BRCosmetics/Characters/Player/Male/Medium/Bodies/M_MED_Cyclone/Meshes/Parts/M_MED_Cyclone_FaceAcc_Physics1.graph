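# Physics asset for the braids on the male medium Cyclone character's face accessory mesh, with capsule bodies for braid bones and upper-body anchors.
# Thirteen braid chains of two to four segments each hang from small MSTR root bodies constrained to the head, which links through neck_01 to spine_05.
# Root, head and neck joints have free swing and twist, while joints along each braid have swing limits and a zero twist limit.
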PhysicsAsset /BRCosmetics/Characters/Player/Male/Medium/Bodies/M_MED_Cyclone/Meshes/Parts/M_MED_Cyclone_FaceAcc_Physics1
bodies {
 dyn_braid_01_l [{shapeType=Capsule, length=1.3691, radius=0.5511, shapeName=dyn_braid_01_l_capsule, center=<2.6159 0 0>, rotation=<rot 90 0 0>}]
 dyn_braid_01_l_MSTR [{shapeType=Capsule, length=0.505, radius=0.505, shapeName=dyn_braid_01_l_MSTR_capsule, rotation=<rot 0 -90 0>}]
 dyn_braid_01_r [{shapeType=Capsule, length=1.0276, radius=0.548, shapeName=dyn_braid_01_r_capsule, center=<2.694 0 0>, rotation=<rot 90 0 0>}]
 dyn_braid_01_r_MSTR [{shapeType=Capsule, length=0.505, radius=0.505, shapeName=dyn_braid_01_r_MSTR_capsule, rotation=<rot 0 -90 0>}]
 dyn_braid_02_l [{shapeType=Capsule, length=1.5667, radius=0.5067, shapeName=dyn_braid_02_l_capsule, center=<1.84 0 0>, rotation=<rot 90 0 0>}]
 dyn_braid_02_r [{shapeType=Capsule, length=1.8026, radius=0.5084, shapeName=dyn_braid_02_r_capsule, center=<1.6049 0 0>, rotation=<rot 90 0 0>}]
 dyn_braid_03_l [{shapeType=Capsule, length=1.7234, radius=0.7151, shapeName=dyn_braid_03_l_capsule, center=<1.4433 0 0>, rotation=<rot 90 0 0>}]
 dyn_braid_03_r [{shapeType=Capsule, length=1.7889, radius=0.685, shapeName=dyn_braid_03_r_capsule, center=<1.4435 0 0>, rotation=<rot 90 0 0>}]
 dyn_braid_04_l [{shapeType=Capsule, length=1.8963, radius=0.8304, shapeName=dyn_braid_04_l_capsule, center=<1.2205 0 0>, rotation=<rot 77.7081 0.0001 0>}]
 dyn_braid_04_r [{shapeType=Capsule, length=1.7292, radius=0.8371, shapeName=dyn_braid_04_r_capsule, center=<1.2208 0 0>, rotation=<rot 78.423 -0.0001 0>}]
 dyn_braid_bk_01 [{shapeType=Capsule, length=1.6314, radius=0.5585, shapeName=dyn_braid_bk_01_capsule, center=<3.3812 0 0>, rotation=<rot 90 0 0>}]
 dyn_braid_bk_01_MSTR [{shapeType=Capsule, length=0.505, radius=0.505, shapeName=dyn_braid_bk_01_MSTR_capsule, rotation=<rot 0 -90 0>}]
 dyn_braid_bk_02 [{shapeType=Capsule, length=2.8269, radius=0.4438, shapeName=dyn_braid_bk_02_capsule, center=<2.5068 0 0>, rotation=<rot 90 0 0>}]
 dyn_braid_bk_03 [{shapeType=Capsule, length=3.2895, radius=0.505, shapeName=dyn_braid_bk_03_capsule, center=<1.8735 0 0>, rotation=<rot 90 0 0>}]
 dyn_braid_bk_bt_01 [{shapeType=Capsule, length=2.3605, radius=0.5295, shapeName=dyn_braid_bk_bt_01_capsule, center=<2.8493 0 0>, rotation=<rot 90 0 0>}]
 dyn_braid_bk_bt_01_MSTR [{shapeType=Capsule, length=0.505, radius=0.505, shapeName=dyn_braid_bk_bt_01_MSTR_capsule, rotation=<rot 0 -90 0>}]
 dyn_braid_bk_bt_01_l [{shapeType=Capsule, length=2.4102, radius=0.4547, shapeName=dyn_braid_bk_bt_01_l_capsule, center=<2.6784 0 0>, rotation=<rot 90 0 0>}]
 dyn_braid_bk_bt_01_l_MSTR [{shapeType=Capsule, length=0.505, radius=0.505, shapeName=dyn_braid_bk_bt_01_l_MSTR_capsule, rotation=<rot 0 -90 0>}]
 dyn_braid_bk_bt_01_r [{shapeType=Capsule, length=1.4739, radius=0.4156, shapeName=dyn_braid_bk_bt_01_r_capsule, center=<3.5461 0 0>, rotation=<rot 90 0 0>}]
 dyn_braid_bk_bt_01_r_MSTR [{shapeType=Capsule, length=0.505, radius=0.505, shapeName=dyn_braid_bk_bt_01_r_MSTR_capsule, rotation=<rot 0 -90 0>}]
 dyn_braid_bk_bt_02 [{shapeType=Capsule, length=3.5193, radius=0.5, shapeName=dyn_braid_bk_bt_02_capsule, center=<1.7127 0 0>, rotation=<rot 90 0 0>}]
 dyn_braid_bk_bt_02_l [{shapeType=Capsule, length=4.0864, radius=0.5, shapeName=dyn_braid_bk_bt_02_l_capsule, center=<2.2322 0 0>, rotation=<rot 90 0 0>}]
 dyn_braid_bk_bt_02_r [{shapeType=Capsule, length=4.1335, radius=0.5, shapeName=dyn_braid_bk_bt_02_r_capsule, center=<2.1466 0 0>, rotation=<rot 90 0 0>}]
 dyn_braid_bk_tp_01 [{shapeType=Capsule, length=2.7553, radius=0.5899, shapeName=dyn_braid_bk_tp_01_capsule, center=<3.3893 0 0>, rotation=<rot 90 0 0>}]
 dyn_braid_bk_tp_01_MSTR [{shapeType=Capsule, length=0.505, radius=0.505, shapeName=dyn_braid_bk_tp_01_MSTR_capsule, rotation=<rot 0 -90 0>}]
 dyn_braid_bk_tp_01_l [{shapeType=Capsule, length=3.1823, radius=0.5158, shapeName=dyn_braid_bk_tp_01_l_capsule, center=<3.5993 0 0>, rotation=<rot 90 0 0>}]
 dyn_braid_bk_tp_01_l_MSTR [{shapeType=Capsule, length=0.505, radius=0.505, shapeName=dyn_braid_bk_tp_01_l_MSTR_capsule, rotation=<rot 0 -90 0>}]
 dyn_braid_bk_tp_01_r [{shapeType=Capsule, length=3.1823, radius=0.5158, shapeName=dyn_braid_bk_tp_01_r_capsule, center=<3.5989 0 0>, rotation=<rot 90 0 0>}]
 dyn_braid_bk_tp_01_r_MSTR [{shapeType=Capsule, length=0.505, radius=0.505, shapeName=dyn_braid_bk_tp_01_r_MSTR_capsule, rotation=<rot 0 -90 0>}]
 dyn_braid_bk_tp_02 [{shapeType=Capsule, length=2.9673, radius=0.5272, shapeName=dyn_braid_bk_tp_02_capsule, center=<2.7949 0 0>, rotation=<rot 90 0 0>}]
 dyn_braid_bk_tp_02_l [{shapeType=Capsule, length=3.9233, radius=0.5, shapeName=dyn_braid_bk_tp_02_l_capsule, center=<2.5311 0 0>, rotation=<rot 90 0 0>}]
 dyn_braid_bk_tp_02_r [{shapeType=Capsule, length=3.9082, radius=0.5, shapeName=dyn_braid_bk_tp_02_r_capsule, center=<2.5314 0 0>, rotation=<rot 90 0 0>}]
 dyn_braid_bk_tp_03 [{
   shapeType=Capsule
   length=3.1863
   radius=0.8201
   shapeName=dyn_braid_bk_tp_03_capsule
   center=<2.2273 0 0>
   rotation=<rot 82.8573 179.9999 180>
  }]
 dyn_braid_bt_01_l [{shapeType=Capsule, length=2.2966, radius=0.5698, shapeName=dyn_braid_bt_01_l_capsule, center=<2.2242 0 0>, rotation=<rot 90 0 0>}]
 dyn_braid_bt_01_l_MSTR [{shapeType=Capsule, length=0.505, radius=0.505, shapeName=dyn_braid_bt_01_l_MSTR_capsule, rotation=<rot 0 -90 0>}]
 dyn_braid_bt_01_r [{shapeType=Capsule, length=2.0897, radius=0.55, shapeName=dyn_braid_bt_01_r_capsule, center=<2.2239 0 0>, rotation=<rot 90 0 0>}]
 dyn_braid_bt_01_r_MSTR [{shapeType=Capsule, length=0.505, radius=0.505, shapeName=dyn_braid_bt_01_r_MSTR_capsule, rotation=<rot 0 -90 0>}]
 dyn_braid_bt_02_l [{shapeType=Capsule, length=2.8085, radius=0.4369, shapeName=dyn_braid_bt_02_l_capsule, center=<2.3035 0 0>, rotation=<rot 90 0 0>}]
 dyn_braid_bt_02_r [{shapeType=Capsule, length=3.1664, radius=0.7002, shapeName=dyn_braid_bt_02_r_capsule, center=<2.3033 0 0>, rotation=<rot 90 0 0>}]
 dyn_braid_fr_01_l [{shapeType=Capsule, length=0.4974, radius=0.4389, shapeName=dyn_braid_fr_01_l_capsule, center=<2.2553 0 0>, rotation=<rot 90 0 0>}]
 dyn_braid_fr_01_l_MSTR [{shapeType=Capsule, length=0.505, radius=0.505, shapeName=dyn_braid_fr_01_l_MSTR_capsule, rotation=<rot 0 -90 0>}]
 dyn_braid_fr_01_r [{shapeType=Capsule, length=0.757, radius=0.5371, shapeName=dyn_braid_fr_01_r_capsule, center=<2.1435 0 0>, rotation=<rot 90 0 0>}]
 dyn_braid_fr_01_r_MSTR [{shapeType=Capsule, length=0.505, radius=0.505, shapeName=dyn_braid_fr_01_r_MSTR_capsule, rotation=<rot 0 -90 0>}]
 dyn_braid_fr_02_l [{shapeType=Capsule, length=1.8242, radius=0.6264, shapeName=dyn_braid_fr_02_l_capsule, center=<1.366 0 0>, rotation=<rot 90 0 0>}]
 dyn_braid_fr_02_r [{shapeType=Capsule, length=1.8103, radius=0.5899, shapeName=dyn_braid_fr_02_r_capsule, center=<1.384 0 0>, rotation=<rot 90 0 0>}]
 dyn_braid_fr_03_l [{shapeType=Capsule, length=1.854, radius=0.4458, shapeName=dyn_braid_fr_03_l_capsule, center=<1.9032 0 0>, rotation=<rot 90 0 0>}]
 dyn_braid_fr_03_r [{shapeType=Capsule, length=1.854, radius=0.4458, shapeName=dyn_braid_fr_03_r_capsule, center=<1.9029 0 0>, rotation=<rot 90 0 0>}]
 dyn_braid_fr_04_l [{shapeType=Capsule, length=0.505, radius=0.505, shapeName=dyn_braid_fr_04_l_capsule, center=<1.216 0 0>, rotation=<rot 90 0 0>}]
 dyn_braid_fr_04_r [{shapeType=Capsule, length=0.505, radius=0.505, shapeName=dyn_braid_fr_04_r_capsule, center=<1.2164 0 0>, rotation=<rot 90 0 0>}]
 head [
  {shapeType=Capsule, length=7.5208, radius=6.3985, shapeName=head_capsule, center=<3.1194 0.6409 0.0001>, rotation=<rot -90 0 0>}
  {shapeType=Capsule, length=8.6519, radius=5.6306, shapeName=head_capsule, center=<2.0079 5.0325 0>, rotation=<rot -90 0.0099 -0.0099>}
 ]
 neck_01 [{shapeType=Capsule, length=14.2452, radius=6.115, shapeName=neck_01_capsule, center=<4.8902 1.3807 0>, rotation=<rot -90 0 -10>}]
 spine_05 [{shapeType=Capsule, length=17.505, radius=8.755, shapeName=spine_05_capsule, center=<3.1326 -1.4854 0>, rotation=<rot 0 -90 0>}]
}
constraints [
 {
  bone1Name=neck_01
  bone2Name=spine_05
  swing1LimitDegrees=45
  swing1Motion=Free
  swing2LimitDegrees=45
  swing2Motion=Free
  twistLimitDegrees=45
  twistMotion=Free
 }
 {
  bone1Name=head
  bone2Name=neck_01
  swing1LimitDegrees=45
  swing1Motion=Free
  swing2LimitDegrees=45
  swing2Motion=Free
  twistLimitDegrees=45
  twistMotion=Free
 }
 {
  bone1Name=dyn_braid_01_l_MSTR
  bone2Name=head
  swing1LimitDegrees=45
  swing1Motion=Free
  swing2LimitDegrees=45
  swing2Motion=Free
  twistLimitDegrees=45
  twistMotion=Free
 }
 {bone1Name=dyn_braid_01_l, bone2Name=dyn_braid_01_l_MSTR, swing1LimitDegrees=45, swing2LimitDegrees=45, twistLimitDegrees=0}
 {bone1Name=dyn_braid_02_l, bone2Name=dyn_braid_01_l, swing1LimitDegrees=45, swing2LimitDegrees=45, twistLimitDegrees=0}
 {bone1Name=dyn_braid_03_l, bone2Name=dyn_braid_02_l, swing1LimitDegrees=45, swing2LimitDegrees=45, twistLimitDegrees=0}
 {bone1Name=dyn_braid_04_l, bone2Name=dyn_braid_03_l, swing1LimitDegrees=45, swing2LimitDegrees=45, twistLimitDegrees=0}
 {
  bone1Name=dyn_braid_01_r_MSTR
  bone2Name=head
  swing1LimitDegrees=45
  swing1Motion=Free
  swing2LimitDegrees=45
  swing2Motion=Free
  twistLimitDegrees=45
  twistMotion=Free
 }
 {bone1Name=dyn_braid_01_r, bone2Name=dyn_braid_01_r_MSTR, swing1LimitDegrees=45, swing2LimitDegrees=45, twistLimitDegrees=0}
 {bone1Name=dyn_braid_02_r, bone2Name=dyn_braid_01_r, swing1LimitDegrees=45, swing2LimitDegrees=45, twistLimitDegrees=0}
 {bone1Name=dyn_braid_03_r, bone2Name=dyn_braid_02_r, swing1LimitDegrees=45, swing2LimitDegrees=45, twistLimitDegrees=0}
 {bone1Name=dyn_braid_04_r, bone2Name=dyn_braid_03_r, swing1LimitDegrees=45, swing2LimitDegrees=45, twistLimitDegrees=0}
 {
  bone1Name=dyn_braid_bk_01_MSTR
  bone2Name=head
  swing1LimitDegrees=45
  swing1Motion=Free
  swing2LimitDegrees=45
  swing2Motion=Free
  twistLimitDegrees=45
  twistMotion=Free
 }
 {bone1Name=dyn_braid_bk_01, bone2Name=dyn_braid_bk_01_MSTR, swing1LimitDegrees=45, swing2LimitDegrees=45, twistLimitDegrees=0}
 {bone1Name=dyn_braid_bk_02, bone2Name=dyn_braid_bk_01, swing1LimitDegrees=45, swing2LimitDegrees=45, twistLimitDegrees=0}
 {bone1Name=dyn_braid_bk_03, bone2Name=dyn_braid_bk_02, swing1LimitDegrees=45, swing2LimitDegrees=45, twistLimitDegrees=0}
 {
  bone1Name=dyn_braid_bk_bt_01_MSTR
  bone2Name=head
  swing1LimitDegrees=45
  swing1Motion=Free
  swing2LimitDegrees=45
  swing2Motion=Free
  twistLimitDegrees=45
  twistMotion=Free
 }
 {bone1Name=dyn_braid_bk_bt_01, bone2Name=dyn_braid_bk_bt_01_MSTR, swing1LimitDegrees=45, swing2LimitDegrees=45, twistLimitDegrees=0}
 {bone1Name=dyn_braid_bk_bt_02, bone2Name=dyn_braid_bk_bt_01, swing1LimitDegrees=45, swing2LimitDegrees=45, twistLimitDegrees=0}
 {
  bone1Name=dyn_braid_bk_bt_01_l_MSTR
  bone2Name=head
  swing1LimitDegrees=45
  swing1Motion=Free
  swing2LimitDegrees=45
  swing2Motion=Free
  twistLimitDegrees=45
  twistMotion=Free
 }
 {bone1Name=dyn_braid_bk_bt_01_l, bone2Name=dyn_braid_bk_bt_01_l_MSTR, swing1LimitDegrees=45, swing2LimitDegrees=45, twistLimitDegrees=0}
 {bone1Name=dyn_braid_bk_bt_02_l, bone2Name=dyn_braid_bk_bt_01_l, swing1LimitDegrees=45, swing2LimitDegrees=45, twistLimitDegrees=0}
 {
  bone1Name=dyn_braid_bk_bt_01_r_MSTR
  bone2Name=head
  swing1LimitDegrees=45
  swing1Motion=Free
  swing2LimitDegrees=45
  swing2Motion=Free
  twistLimitDegrees=45
  twistMotion=Free
 }
 {bone1Name=dyn_braid_bk_bt_01_r, bone2Name=dyn_braid_bk_bt_01_r_MSTR, swing1LimitDegrees=45, swing2LimitDegrees=45, twistLimitDegrees=0}
 {bone1Name=dyn_braid_bk_bt_02_r, bone2Name=dyn_braid_bk_bt_01_r, swing1LimitDegrees=45, swing2LimitDegrees=45, twistLimitDegrees=0}
 {
  bone1Name=dyn_braid_bk_tp_01_MSTR
  bone2Name=head
  swing1LimitDegrees=45
  swing1Motion=Free
  swing2LimitDegrees=45
  swing2Motion=Free
  twistLimitDegrees=45
  twistMotion=Free
 }
 {bone1Name=dyn_braid_bk_tp_01, bone2Name=dyn_braid_bk_tp_01_MSTR, swing1LimitDegrees=45, swing2LimitDegrees=45, twistLimitDegrees=0}
 {bone1Name=dyn_braid_bk_tp_02, bone2Name=dyn_braid_bk_tp_01, swing1LimitDegrees=45, swing2LimitDegrees=45, twistLimitDegrees=0}
 {bone1Name=dyn_braid_bk_tp_03, bone2Name=dyn_braid_bk_tp_02, swing1LimitDegrees=45, swing2LimitDegrees=45, twistLimitDegrees=0}
 {
  bone1Name=dyn_braid_bk_tp_01_l_MSTR
  bone2Name=head
  swing1LimitDegrees=45
  swing1Motion=Free
  swing2LimitDegrees=45
  swing2Motion=Free
  twistLimitDegrees=45
  twistMotion=Free
 }
 {bone1Name=dyn_braid_bk_tp_01_l, bone2Name=dyn_braid_bk_tp_01_l_MSTR, swing1LimitDegrees=45, swing2LimitDegrees=45, twistLimitDegrees=0}
 {bone1Name=dyn_braid_bk_tp_02_l, bone2Name=dyn_braid_bk_tp_01_l, swing1LimitDegrees=45, swing2LimitDegrees=45, twistLimitDegrees=0}
 {
  bone1Name=dyn_braid_bk_tp_01_r_MSTR
  bone2Name=head
  swing1LimitDegrees=45
  swing1Motion=Free
  swing2LimitDegrees=45
  swing2Motion=Free
  twistLimitDegrees=45
  twistMotion=Free
 }
 {bone1Name=dyn_braid_bk_tp_01_r, bone2Name=dyn_braid_bk_tp_01_r_MSTR, swing1LimitDegrees=45, swing2LimitDegrees=45, twistLimitDegrees=0}
 {bone1Name=dyn_braid_bk_tp_02_r, bone2Name=dyn_braid_bk_tp_01_r, swing1LimitDegrees=45, swing2LimitDegrees=45, twistLimitDegrees=0}
 {
  bone1Name=dyn_braid_bt_01_l_MSTR
  bone2Name=head
  swing1LimitDegrees=45
  swing1Motion=Free
  swing2LimitDegrees=45
  swing2Motion=Free
  twistLimitDegrees=45
  twistMotion=Free
 }
 {bone1Name=dyn_braid_bt_01_l, bone2Name=dyn_braid_bt_01_l_MSTR, swing1LimitDegrees=45, swing2LimitDegrees=45, twistLimitDegrees=0}
 {bone1Name=dyn_braid_bt_02_l, bone2Name=dyn_braid_bt_01_l, swing1LimitDegrees=45, swing2LimitDegrees=45, twistLimitDegrees=0}
 {
  bone1Name=dyn_braid_bt_01_r_MSTR
  bone2Name=head
  swing1LimitDegrees=45
  swing1Motion=Free
  swing2LimitDegrees=45
  swing2Motion=Free
  twistLimitDegrees=45
  twistMotion=Free
 }
 {bone1Name=dyn_braid_bt_01_r, bone2Name=dyn_braid_bt_01_r_MSTR, swing1LimitDegrees=45, swing2LimitDegrees=45, twistLimitDegrees=0}
 {bone1Name=dyn_braid_bt_02_r, bone2Name=dyn_braid_bt_01_r, swing1LimitDegrees=45, swing2LimitDegrees=45, twistLimitDegrees=0}
 {
  bone1Name=dyn_braid_fr_01_l_MSTR
  bone2Name=head
  swing1LimitDegrees=45
  swing1Motion=Free
  swing2LimitDegrees=45
  swing2Motion=Free
  twistLimitDegrees=45
  twistMotion=Free
 }
 {bone1Name=dyn_braid_fr_01_l, bone2Name=dyn_braid_fr_01_l_MSTR, swing1LimitDegrees=45, swing2LimitDegrees=45, twistLimitDegrees=0}
 {bone1Name=dyn_braid_fr_02_l, bone2Name=dyn_braid_fr_01_l, swing1LimitDegrees=45, swing2LimitDegrees=45, twistLimitDegrees=0}
 {bone1Name=dyn_braid_fr_03_l, bone2Name=dyn_braid_fr_02_l, swing1LimitDegrees=45, swing2LimitDegrees=45, twistLimitDegrees=0}
 {bone1Name=dyn_braid_fr_04_l, bone2Name=dyn_braid_fr_03_l, swing1LimitDegrees=45, swing2LimitDegrees=45, twistLimitDegrees=0}
 {
  bone1Name=dyn_braid_fr_01_r_MSTR
  bone2Name=head
  swing1LimitDegrees=45
  swing1Motion=Free
  swing2LimitDegrees=45
  swing2Motion=Free
  twistLimitDegrees=45
  twistMotion=Free
 }
 {bone1Name=dyn_braid_fr_01_r, bone2Name=dyn_braid_fr_01_r_MSTR, swing1LimitDegrees=45, swing2LimitDegrees=45, twistLimitDegrees=0}
 {bone1Name=dyn_braid_fr_02_r, bone2Name=dyn_braid_fr_01_r, swing1LimitDegrees=45, swing2LimitDegrees=45, twistLimitDegrees=0}
 {bone1Name=dyn_braid_fr_03_r, bone2Name=dyn_braid_fr_02_r, swing1LimitDegrees=45, swing2LimitDegrees=45, twistLimitDegrees=0}
 {bone1Name=dyn_braid_fr_04_r, bone2Name=dyn_braid_fr_03_r, swing1LimitDegrees=45, swing2LimitDegrees=45, twistLimitDegrees=0}
]
