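# Anim Blueprint for the Cyclone male medium FaceAcc part, adding secondary motion to its braid bones over a pose copied from the attached parent mesh.
# The cached pose feeds a RigidBody simulation in neck_01 space, using WindVector as gravity override, and a chain of Trail controllers with rotation and planar limits.
# bDisableRigidBodies selects the output: trail result when true, rigid body result otherwise.
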AnimBlueprint /BRCosmetics/Characters/Player/Male/Medium/Bodies/M_MED_Cyclone/Meshes/Parts/M_MED_Cyclone_FaceAcc_AnimBP
target_skeleton=&/M_MED_Cyclone_FaceAcc_Skeleton
graphs [
 AnimGraph {nodes [
   Output Pose(Result=Blend Poses by bool(
     BlendPose_0=Use cached pose 'trailControlCache'()
     BlendPose_1=Use cached pose 'rbanCache'()
     bActiveValue=Get bDisableRigidBodies() -> (bDisableRigidBodies:bool)
    ) {blendType=Linear}) {layerGroup=None}
   Save cached pose 'initCache'(Pose=Copy Pose From Mesh() {bUseAttachedParent=true}) {cacheName=initCache}
   n1 = Trail controller - Bone: dyn_braid_fr_04_l!(
    ComponentPose=Local To Component(LocalPose=Use cached pose 'initCache'())
    Alpha=1
    RelaxationSpeedScale:float=1.5
    FakeVelocity:Vector="0,0,0"
   ) -> (Pose) {
    bActorSpaceFakeVel=true
    bLimitRotation=true
    bUsePlanarLimit=true
    chainLength=4
    debugLifeTime=0.1
    maxDeltaTime=0.0333
    relaxationSpeedScale=1.2
    alphaScaleBiasClamp {bInterpResult=true, interpSpeedDecreasing=25, interpSpeedIncreasing=25, scale=1.5}
    baseJoint {boneName=root}
    planarLimits [
     {drivingBone {boneName=root}, planeTransform {location=<0 0 10>, rotation=<rot 0 0 0>, scale=<1 1 1>}}
     {drivingBone {boneName=spine_04}, planeTransform {location=<0 40 0>, rotation=<rot 0 0 -90>, scale=<1 1 1>}}
    ]
    rotationLimits [
     {limitMax=<0 0 30>, limitMin=<0 -30 -30>}
     {limitMax=<0 0 30>, limitMin=<0 -30 -30>}
     {limitMax=<0 0 30>, limitMin=<0 -30 -30>}
     {limitMax=<180 180 180>}
    ]
    rotationOffsets [<0 0 0>, <0 0 0>, <0 0 0>, <0 0 0>]
    trailBone {boneName=dyn_braid_fr_04_l}
    trailRelaxationSpeed {editorCurveData {keys [<at 0 15 RCIM_Linear>,<at 0.5289 13.1838 RCIM_Linear>,<at 1 6 RCIM_Linear>]}}
   }
   n2 = Trail controller - Bone: dyn_braid_04_l!(ComponentPose=@n1, Alpha=1, RelaxationSpeedScale:float=1.5, FakeVelocity:Vector="0,0,0") -> (Pose) {
    bActorSpaceFakeVel=true
    bLimitRotation=true
    bUsePlanarLimit=true
    chainLength=4
    debugLifeTime=0.1
    maxDeltaTime=0.0333
    relaxationSpeedScale=1.2
    alphaScaleBiasClamp {bInterpResult=true, interpSpeedDecreasing=25, interpSpeedIncreasing=25, scale=1.5}
    baseJoint {boneName=root}
    planarLimits [
     {drivingBone {boneName=root}, planeTransform {location=<0 0 10>, rotation=<rot 0 0 0>, scale=<1 1 1>}}
     {drivingBone {boneName=spine_04}, planeTransform {location=<0 40 0>, rotation=<rot 0 0 -90>, scale=<1 1 1>}}
    ]
    rotationLimits [
     {limitMax=<0 0 30>, limitMin=<0 -30 -30>}
     {limitMax=<0 0 30>, limitMin=<0 -30 -30>}
     {limitMax=<0 0 30>, limitMin=<0 -30 -30>}
     {limitMax=<180 180 180>}
    ]
    rotationOffsets [<0 0 0>, <0 0 0>, <0 0 0>, <0 0 0>]
    trailBone {boneName=dyn_braid_04_l}
    trailRelaxationSpeed {editorCurveData {keys [<at 0 15 RCIM_Linear>,<at 0.5289 13.1838 RCIM_Linear>,<at 1 6 RCIM_Linear>]}}
   }
   n3 = Trail controller - Bone: dyn_braid_bk_bt_02_l!(ComponentPose=@n2, Alpha=1, RelaxationSpeedScale:float=1.5, FakeVelocity:Vector="0,0,0") -> (Pose) {
    bActorSpaceFakeVel=true
    bLimitRotation=true
    bUsePlanarLimit=true
    debugLifeTime=0.1
    maxDeltaTime=0.0333
    relaxationSpeedScale=1.2
    alphaScaleBiasClamp {bInterpResult=true, interpSpeedDecreasing=25, interpSpeedIncreasing=25, scale=1.5}
    baseJoint {boneName=root}
    planarLimits [
     {drivingBone {boneName=root}, planeTransform {location=<0 0 10>, rotation=<rot 0 0 0>, scale=<1 1 1>}}
     {drivingBone {boneName=spine_04}, planeTransform {location=<0 40 0>, rotation=<rot 0 0 -90>, scale=<1 1 1>}}
    ]
    rotationLimits [{limitMax=<0 0 30>, limitMin=<0 -30 -30>}, {limitMax=<0 0 30>, limitMin=<0 -30 -30>}]
    rotationOffsets [<0 0 0>, <0 0 0>]
    trailBone {boneName=dyn_braid_bk_bt_02_l}
    trailRelaxationSpeed {editorCurveData {keys [<at 0 15 RCIM_Linear>,<at 0.5289 13.1838 RCIM_Linear>,<at 1 6 RCIM_Linear>]}}
   }
   n4 = Trail controller - Bone: dyn_braid_bk_tp_02_l!(ComponentPose=@n3, Alpha=1, RelaxationSpeedScale:float=1.5, FakeVelocity:Vector="0,0,0") -> (Pose) {
    bActorSpaceFakeVel=true
    bLimitRotation=true
    bUsePlanarLimit=true
    debugLifeTime=0.1
    maxDeltaTime=0.0333
    relaxationSpeedScale=1.2
    alphaScaleBiasClamp {bInterpResult=true, interpSpeedDecreasing=25, interpSpeedIncreasing=25, scale=1.5}
    baseJoint {boneName=root}
    planarLimits [
     {drivingBone {boneName=root}, planeTransform {location=<0 0 10>, rotation=<rot 0 0 0>, scale=<1 1 1>}}
     {drivingBone {boneName=spine_04}, planeTransform {location=<0 40 0>, rotation=<rot 0 0 -90>, scale=<1 1 1>}}
    ]
    rotationLimits [{limitMax=<0 0 30>, limitMin=<0 -30 -30>}, {limitMax=<0 0 30>, limitMin=<0 -30 -30>}]
    rotationOffsets [<0 0 0>, <0 0 0>]
    trailBone {boneName=dyn_braid_bk_tp_02_l}
    trailRelaxationSpeed {editorCurveData {keys [<at 0 15 RCIM_Linear>,<at 0.5289 13.1838 RCIM_Linear>,<at 1 6 RCIM_Linear>]}}
   }
   n5 = Trail controller - Bone: dyn_braid_bt_02_l!(ComponentPose=@n4, Alpha=1, RelaxationSpeedScale:float=1.5, FakeVelocity:Vector="0,0,0") -> (Pose) {
    bActorSpaceFakeVel=true
    bLimitRotation=true
    bUsePlanarLimit=true
    debugLifeTime=0.1
    maxDeltaTime=0.0333
    relaxationSpeedScale=1.2
    alphaScaleBiasClamp {bInterpResult=true, interpSpeedDecreasing=25, interpSpeedIncreasing=25, scale=1.5}
    baseJoint {boneName=root}
    planarLimits [
     {drivingBone {boneName=root}, planeTransform {location=<0 0 10>, rotation=<rot 0 0 0>, scale=<1 1 1>}}
     {drivingBone {boneName=spine_04}, planeTransform {location=<0 40 0>, rotation=<rot 0 0 -90>, scale=<1 1 1>}}
    ]
    rotationLimits [{limitMax=<0 0 30>, limitMin=<0 -30 -30>}, {limitMax=<0 0 30>, limitMin=<0 -30 -30>}]
    rotationOffsets [<0 0 0>, <0 0 0>]
    trailBone {boneName=dyn_braid_bt_02_l}
    trailRelaxationSpeed {editorCurveData {keys [<at 0 15 RCIM_Linear>,<at 0.5289 13.1838 RCIM_Linear>,<at 1 6 RCIM_Linear>]}}
   }
   n6 = Trail controller - Bone: dyn_braid_fr_04_r!(ComponentPose=@n5, Alpha=1, RelaxationSpeedScale:float=1.5, FakeVelocity:Vector="0,0,0") -> (Pose) {
    bActorSpaceFakeVel=true
    bLimitRotation=true
    bUsePlanarLimit=true
    chainLength=4
    debugLifeTime=0.1
    maxDeltaTime=0.0333
    relaxationSpeedScale=1.2
    alphaScaleBiasClamp {bInterpResult=true, interpSpeedDecreasing=25, interpSpeedIncreasing=25, scale=1.5}
    baseJoint {boneName=root}
    planarLimits [
     {drivingBone {boneName=root}, planeTransform {location=<0 0 10>, rotation=<rot 0 0 0>, scale=<1 1 1>}}
     {drivingBone {boneName=spine_04}, planeTransform {location=<0 40 0>, rotation=<rot 0 0 -90>, scale=<1 1 1>}}
    ]
    rotationLimits [
     {limitMax=<0 0 30>, limitMin=<0 -30 -30>}
     {limitMax=<0 0 30>, limitMin=<0 -30 -30>}
     {limitMax=<0 0 30>, limitMin=<0 -30 -30>}
     {limitMax=<180 180 180>}
    ]
    rotationOffsets [<0 0 0>, <0 0 0>, <0 0 0>, <0 0 0>]
    trailBone {boneName=dyn_braid_fr_04_r}
    trailRelaxationSpeed {editorCurveData {keys [<at 0 15 RCIM_Linear>,<at 0.5289 13.1838 RCIM_Linear>,<at 1 6 RCIM_Linear>]}}
   }
   n7 = Trail controller - Bone: dyn_braid_04_r!(ComponentPose=@n6, Alpha=1, RelaxationSpeedScale:float=1.5, FakeVelocity:Vector="0,0,0") -> (Pose) {
    bActorSpaceFakeVel=true
    bLimitRotation=true
    bUsePlanarLimit=true
    chainLength=4
    debugLifeTime=0.1
    maxDeltaTime=0.0333
    relaxationSpeedScale=1.2
    alphaScaleBiasClamp {bInterpResult=true, interpSpeedDecreasing=25, interpSpeedIncreasing=25, scale=1.5}
    baseJoint {boneName=root}
    planarLimits [
     {drivingBone {boneName=root}, planeTransform {location=<0 0 10>, rotation=<rot 0 0 0>, scale=<1 1 1>}}
     {drivingBone {boneName=spine_04}, planeTransform {location=<0 40 0>, rotation=<rot 0 0 -90>, scale=<1 1 1>}}
    ]
    rotationLimits [
     {limitMax=<0 0 30>, limitMin=<0 -30 -30>}
     {limitMax=<0 0 30>, limitMin=<0 -30 -30>}
     {limitMax=<0 0 30>, limitMin=<0 -30 -30>}
     {limitMax=<180 180 180>}
    ]
    rotationOffsets [<0 0 0>, <0 0 0>, <0 0 0>, <0 0 0>]
    trailBone {boneName=dyn_braid_04_r}
    trailRelaxationSpeed {editorCurveData {keys [<at 0 15 RCIM_Linear>,<at 0.5289 13.1838 RCIM_Linear>,<at 1 6 RCIM_Linear>]}}
   }
   n8 = Trail controller - Bone: dyn_braid_bk_bt_02_r!(ComponentPose=@n7, Alpha=1, RelaxationSpeedScale:float=1.5, FakeVelocity:Vector="0,0,0") -> (Pose) {
    bActorSpaceFakeVel=true
    bLimitRotation=true
    bUsePlanarLimit=true
    debugLifeTime=0.1
    maxDeltaTime=0.0333
    relaxationSpeedScale=1.2
    alphaScaleBiasClamp {bInterpResult=true, interpSpeedDecreasing=25, interpSpeedIncreasing=25, scale=1.5}
    baseJoint {boneName=root}
    planarLimits [
     {drivingBone {boneName=root}, planeTransform {location=<0 0 10>, rotation=<rot 0 0 0>, scale=<1 1 1>}}
     {drivingBone {boneName=spine_04}, planeTransform {location=<0 40 0>, rotation=<rot 0 0 -90>, scale=<1 1 1>}}
    ]
    rotationLimits [{limitMax=<0 0 30>, limitMin=<0 -30 -30>}, {limitMax=<0 0 30>, limitMin=<0 -30 -30>}]
    rotationOffsets [<0 0 0>, <0 0 0>]
    trailBone {boneName=dyn_braid_bk_bt_02_r}
    trailRelaxationSpeed {editorCurveData {keys [<at 0 15 RCIM_Linear>,<at 0.5289 13.1838 RCIM_Linear>,<at 1 6 RCIM_Linear>]}}
   }
   n9 = Trail controller - Bone: dyn_braid_bk_tp_02_r!(ComponentPose=@n8, Alpha=1, RelaxationSpeedScale:float=1.5, FakeVelocity:Vector="0,0,0") -> (Pose) {
    bActorSpaceFakeVel=true
    bLimitRotation=true
    bUsePlanarLimit=true
    debugLifeTime=0.1
    maxDeltaTime=0.0333
    relaxationSpeedScale=1.2
    alphaScaleBiasClamp {bInterpResult=true, interpSpeedDecreasing=25, interpSpeedIncreasing=25, scale=1.5}
    baseJoint {boneName=root}
    planarLimits [
     {drivingBone {boneName=root}, planeTransform {location=<0 0 10>, rotation=<rot 0 0 0>, scale=<1 1 1>}}
     {drivingBone {boneName=spine_04}, planeTransform {location=<0 40 0>, rotation=<rot 0 0 -90>, scale=<1 1 1>}}
    ]
    rotationLimits [{limitMax=<0 0 30>, limitMin=<0 -30 -30>}, {limitMax=<0 0 30>, limitMin=<0 -30 -30>}]
    rotationOffsets [<0 0 0>, <0 0 0>]
    trailBone {boneName=dyn_braid_bk_tp_02_r}
    trailRelaxationSpeed {editorCurveData {keys [<at 0 15 RCIM_Linear>,<at 0.5289 13.1838 RCIM_Linear>,<at 1 6 RCIM_Linear>]}}
   }
   n10 = Trail controller - Bone: dyn_braid_bt_02_r!(ComponentPose=@n9, Alpha=1, RelaxationSpeedScale:float=1.5, FakeVelocity:Vector="0,0,0") -> (Pose) {
    bActorSpaceFakeVel=true
    bLimitRotation=true
    bUsePlanarLimit=true
    debugLifeTime=0.1
    maxDeltaTime=0.0333
    relaxationSpeedScale=1.2
    alphaScaleBiasClamp {bInterpResult=true, interpSpeedDecreasing=25, interpSpeedIncreasing=25, scale=1.5}
    baseJoint {boneName=root}
    planarLimits [
     {drivingBone {boneName=root}, planeTransform {location=<0 0 10>, rotation=<rot 0 0 0>, scale=<1 1 1>}}
     {drivingBone {boneName=spine_04}, planeTransform {location=<0 40 0>, rotation=<rot 0 0 -90>, scale=<1 1 1>}}
    ]
    rotationLimits [{limitMax=<0 0 30>, limitMin=<0 -30 -30>}, {limitMax=<0 0 30>, limitMin=<0 -30 -30>}]
    rotationOffsets [<0 0 0>, <0 0 0>]
    trailBone {boneName=dyn_braid_bt_02_r}
    trailRelaxationSpeed {editorCurveData {keys [<at 0 15 RCIM_Linear>,<at 0.5289 13.1838 RCIM_Linear>,<at 1 6 RCIM_Linear>]}}
   }
   n11 = Trail controller - Bone: dyn_braid_bk_03!(ComponentPose=@n10, Alpha=1, RelaxationSpeedScale:float=1.5, FakeVelocity:Vector="0,0,0") -> (Pose) {
    bActorSpaceFakeVel=true
    bLimitRotation=true
    bUsePlanarLimit=true
    chainLength=3
    debugLifeTime=0.1
    maxDeltaTime=0.0333
    relaxationSpeedScale=1.2
    alphaScaleBiasClamp {bInterpResult=true, interpSpeedDecreasing=25, interpSpeedIncreasing=25, scale=1.5}
    baseJoint {boneName=root}
    planarLimits [
     {drivingBone {boneName=root}, planeTransform {location=<0 0 10>, rotation=<rot 0 0 0>, scale=<1 1 1>}}
     {drivingBone {boneName=spine_04}, planeTransform {location=<0 40 0>, rotation=<rot 0 0 -90>, scale=<1 1 1>}}
    ]
    rotationLimits [{limitMax=<0 0 30>, limitMin=<0 -30 -30>}, {limitMax=<0 0 30>, limitMin=<0 -30 -30>}, {limitMax=<180 180 180>}]
    rotationOffsets [<0 0 0>, <0 0 0>, <0 0 0>]
    trailBone {boneName=dyn_braid_bk_03}
    trailRelaxationSpeed {editorCurveData {keys [<at 0 15 RCIM_Linear>,<at 0.5289 13.1838 RCIM_Linear>,<at 1 6 RCIM_Linear>]}}
   }
   n12 = Trail controller - Bone: dyn_braid_bk_bt_02!(ComponentPose=@n11, Alpha=1, RelaxationSpeedScale:float=1.5, FakeVelocity:Vector="0,0,0") -> (Pose) {
    bActorSpaceFakeVel=true
    bLimitRotation=true
    bUsePlanarLimit=true
    debugLifeTime=0.1
    maxDeltaTime=0.0333
    relaxationSpeedScale=1.2
    alphaScaleBiasClamp {bInterpResult=true, interpSpeedDecreasing=25, interpSpeedIncreasing=25, scale=1.5}
    baseJoint {boneName=root}
    planarLimits [
     {drivingBone {boneName=root}, planeTransform {location=<0 0 10>, rotation=<rot 0 0 0>, scale=<1 1 1>}}
     {drivingBone {boneName=spine_04}, planeTransform {location=<0 40 0>, rotation=<rot 0 0 -90>, scale=<1 1 1>}}
    ]
    rotationLimits [{limitMax=<0 0 30>, limitMin=<0 -30 -30>}, {limitMax=<0 0 30>, limitMin=<0 -30 -30>}]
    rotationOffsets [<0 0 0>, <0 0 0>]
    trailBone {boneName=dyn_braid_bk_bt_02}
    trailRelaxationSpeed {editorCurveData {keys [<at 0 15 RCIM_Linear>,<at 0.5289 13.1838 RCIM_Linear>,<at 1 6 RCIM_Linear>]}}
   }
   n13 = Trail controller - Bone: dyn_braid_bk_tp_03!(ComponentPose=@n12, Alpha=1, RelaxationSpeedScale:float=1.5, FakeVelocity:Vector="0,0,0") -> (Pose) {
    bActorSpaceFakeVel=true
    bLimitRotation=true
    bUsePlanarLimit=true
    chainLength=3
    debugLifeTime=0.1
    maxDeltaTime=0.0333
    relaxationSpeedScale=1.2
    alphaScaleBiasClamp {bInterpResult=true, interpSpeedDecreasing=25, interpSpeedIncreasing=25, scale=1.5}
    baseJoint {boneName=root}
    planarLimits [
     {drivingBone {boneName=root}, planeTransform {location=<0 0 10>, rotation=<rot 0 0 0>, scale=<1 1 1>}}
     {drivingBone {boneName=spine_04}, planeTransform {location=<0 40 0>, rotation=<rot 0 0 -90>, scale=<1 1 1>}}
    ]
    rotationLimits [{limitMax=<0 0 30>, limitMin=<0 -30 -30>}, {limitMax=<0 0 30>, limitMin=<0 -30 -30>}, {limitMax=<180 180 180>}]
    rotationOffsets [<0 0 0>, <0 0 0>, <0 0 0>]
    trailBone {boneName=dyn_braid_bk_tp_03}
    trailRelaxationSpeed {editorCurveData {keys [<at 0 15 RCIM_Linear>,<at 0.5289 13.1838 RCIM_Linear>,<at 1 6 RCIM_Linear>]}}
   }
   Save cached pose 'trailControlCache'(Pose=Component To Local(ComponentPose=@n13)) {cacheName=trailControlCache}
   n14 = RigidBody!(
    ComponentPose=Local To Component(LocalPose=Use cached pose 'initCache'())
    Alpha
    OverrideWorldGravity:Vector=Get WindVector() -> (WindVector:Vector)
    ExternalForce
   ) -> (Pose) {
    bClampLinearTranslationLimitToRefPose=true
    bOverrideWorldGravity=true
    overridePhysicsAsset=&/M_MED_Cyclone_FaceAcc_Physics1
    simulationSpace=BaseBoneSpace
    alphaBoolBlend {blendOption=Linear}
    baseBoneRef {boneName=neck_01}
    componentAppliedLinearAccClamp=<3000 3000 3000>
    componentLinearAccScale=<0.4 0.4 0.7>
    componentLinearVelScale=<0.4 0.4 0.5>
    overrideWorldGravity {z=-980}
   }
   Save cached pose 'rbanCache'(Pose=Component To Local(ComponentPose=@n14)) {cacheName=rbanCache}
  ]}
 EventGraph {links [], nodes []}
]
variables []
